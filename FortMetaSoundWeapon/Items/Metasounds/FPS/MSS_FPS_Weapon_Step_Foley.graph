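# One-shot stereo MetaSound for first-person weapon step foley, randomly playing metal armor swap-weapon clips with local-only attenuation.
# On play, a PlayerRelation check against zero either starts the random-sound subgraph or fires OnFinished immediately.
# Stance routes five presets of pitch, volume and high-pass cutoff, each smoothed by InterpTo.
# The filtered mono signal is scaled by OutputVolume, then stereo-panned by PanAmount.
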MetaSoundSource /FortMetaSoundWeapon/Items/Metasounds/FPS/MSS_FPS_Weapon_Step_Foley
inputs [
 OutputVolume {type=Float}
 PanAmount {type=Float}
 PlayerRelation {type=Int32}
 Random.PitchShift.Max {type=Float}
 Random.PitchShift.Min {type=Float}
 Random.Volume.Max {type=Float}
 Random.Volume.Min {type=Float}
 Sounds {type=WaveAsset:Array}
 Stance {type=Int32}
 UE.Source.OnPlay {type=Trigger}
]
interfaces [UE.OutputFormat.Stereo:1.0, UE.Source.OneShot:1.0, UE.Source:1.1]
outputs [
 "UE.OutputFormat.Stereo.Audio:0" {type=Audio}
 "UE.OutputFormat.Stereo.Audio:1" {type=Audio}
 UE.Source.OneShot.OnFinished {type=Trigger}
]
pages [Default {
  input_defaults {
   OutputVolume=0.15
   PanAmount=null
   PlayerRelation=null
   Random.PitchShift.Max=1.05
   Random.PitchShift.Min=0.95
   Random.Volume.Max=null
   Random.Volume.Min=null
   Stance=null
   UE.Source.OnPlay=null
   Sounds [
    /Game/Sounds/Fort_Player/Clothing/MetalArmor/Waves/Foley_Clothing_MetalArmor_SwapWeapon_01
    /Game/Sounds/Fort_Player/Clothing/MetalArmor/Waves/Foley_Clothing_MetalArmor_SwapWeapon_02
    /Game/Sounds/Fort_Player/Clothing/MetalArmor/Waves/Foley_Clothing_MetalArmor_SwapWeapon_03
    /Game/Sounds/Fort_Player/Clothing/MetalArmor/Waves/Foley_Clothing_MetalArmor_SwapWeapon_04
    /Game/Sounds/Fort_Player/Clothing/MetalArmor/Waves/Foley_Clothing_MetalArmor_SwapWeapon_05
   ]
  }
  nodes [
   n1 = TriggerCompare.Clamp:Int32(A=^PlayerRelation, Compare=^UE.Source.OnPlay) {also {defaults {B=0}}}
   n2 = InitVariable.Float()
   TriggerSelect5_0 = "UE.Trigger Select (5)"(Index=^Stance, In=^UE.Source.OnPlay)
   TriggerRouteTriggerRouteFloat5_1 = "TriggerRoute.Trigger Route (Float, 5):Float"(
    Set 0=@TriggerSelect5_0.Out 0
    Set 1=@TriggerSelect5_0.Out 1
    Set 2=@TriggerSelect5_0.Out 2
    Set 3=@TriggerSelect5_0.Out 3
    Set 4=@TriggerSelect5_0.Out 4
   ) {also {defaults {Value 0=1, Value 1=1, Value 2=0.9, Value 3=1, Value 4=1.05}}}
   n3 = UE.InterpTo:Audio(Target=@TriggerRouteTriggerRouteFloat5_1.Value) {also {defaults {Interp Time=0.3}}}
   n4 = InitVariable.Float()
   TriggerRouteTriggerRouteFloat5_2 = "TriggerRoute.Trigger Route (Float, 5):Float"(
    Set 0=@TriggerSelect5_0.Out 0
    Set 1=@TriggerSelect5_0.Out 1
    Set 2=@TriggerSelect5_0.Out 2
    Set 3=@TriggerSelect5_0.Out 3
    Set 4=@TriggerSelect5_0.Out 4
   ) {also {defaults {Value 1=0.5, Value 2=1, Value 4=0.2}}}
   n5 = UE.InterpTo:Audio(Target=@TriggerRouteTriggerRouteFloat5_2.Value) {also {defaults {Interp Time=0.3}}}
   n6 = &MSS_PlayRandomSound_Mono(
    MaxPitch=^Random.PitchShift.Max
    MinPitch=^Random.PitchShift.Min
    MaxVolume=^Random.Volume.Max
    MinVolume=^Random.Volume.Min
    Sounds=^Sounds
    UE.Source.OnPlay=@n1.True
    Pitch=VariableAccessor.Float(Variable=VariableMutator.Float(Variable=@n2.Variable, Value=@n3.Value).Variable).Value
    Volume=VariableAccessor.Float(Variable=VariableMutator.Float(Variable=@n4.Variable, Value=@n5.Value).Variable).Value
   )
   n7 = InitVariable.Float()
   TriggerRouteTriggerRouteFloat5_0 = "TriggerRoute.Trigger Route (Float, 5):Float"(
    Set 0=@TriggerSelect5_0.Out 0
    Set 1=@TriggerSelect5_0.Out 1
    Set 2=@TriggerSelect5_0.Out 2
    Set 3=@TriggerSelect5_0.Out 3
    Set 4=@TriggerSelect5_0.Out 4
   ) {also {defaults {Value 0=200, Value 1=200, Value 2=0, Value 3=300, Value 4=300}}}
   n8 = UE.InterpTo:Audio(Target=@TriggerRouteTriggerRouteFloat5_0.Value) {also {defaults {Interp Time=0.3}}}
   OnePoleHighPassFilter_0 = UE.One-Pole High Pass Filter:Audio(
    In=@n6."UE.OutputFormat.Mono.Audio:0"
    Cutoff Frequency=VariableAccessor.Float(Variable=VariableMutator.Float(Variable=@n7.Variable, Value=@n8.Value).Variable).Value
   )
   StereoPanner_0 = UE.Stereo Panner(Pan Amount=^PanAmount, In=UE.Multiply:Audio by Float(AdditionalOperands=^OutputVolume, PrimaryOperand=@OnePoleHighPassFilter_0.Out).Out)
   TriggerAnyTriggerAny2_0 = "TriggerAny.Trigger Any (2)"(In 0=@n6.UE.Source.OneShot.OnFinished, In 1=@n1.False)
  ]
  links [
   {dst="@UE.OutputFormat.Stereo.Audio:0", src=StereoPanner_0.Out Left}
   {dst="@UE.OutputFormat.Stereo.Audio:1", src=StereoPanner_0.Out Right}
   {dst="@UE.Source.OneShot.OnFinished", src=TriggerAnyTriggerAny2_0.Out}
  ]
 }]
properties {
 assetImportData=&self:AssetImportData
 attenuationSettings=/FortMetaSoundWeapon/Items/Shared/Settings/FPS/AS_Weapon_FP_Foley_LocalOnly
 maxDistance=500
 numChannels=2
 outputFormat=Stereo
 sampleRate=48000
 soundClassObject=&Fort_WeaponFoley
 concurrencySet [&CS_Fort_WeaponFoley]
}
