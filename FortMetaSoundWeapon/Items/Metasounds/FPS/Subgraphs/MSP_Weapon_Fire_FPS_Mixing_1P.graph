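# MetaSound patch that mixes first-person weapon fire layers (main, mech, low ammo, stop, sweetener, shells, suppressor) into six surround outputs.
# Control-bus modulators scale layer volumes, while a fire-rate envelope and an enclosure-driven crossfade to a per-gunshot worldization envelope shape overall gain.
# An enabled suppressor adds its stereo layer and envelope-shortened base weapon audio.
# The Mobile page reduces this to stereo main, mech and low-ammo layers.
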MetaSoundPatch /FortMetaSoundWeapon/Items/Metasounds/FPS/Subgraphs/MSP_Weapon_Fire_FPS_Mixing_1P
inputs [
 LowAmmo.C {type=Audio}
 LowAmmo.Volume {type=Float}
 MainFire.C {type=Audio}
 MainFire.L {type=Audio}
 MainFire.LFE {type=Audio}
 MainFire.Ls {type=Audio}
 MainFire.R {type=Audio}
 MainFire.Rs {type=Audio}
 MainFire.Volume {type=Float}
 Mech.C {type=Audio}
 Mech.L {type=Audio}
 Mech.R {type=Audio}
 Mech.Volume {type=Float}
 OnGunshot {type=Trigger}
 OnStopFire.L {type=Audio}
 OnStopFire.R {type=Audio}
 OnStopFire.Volume {type=Float}
 Shells.L {type=Audio}
 Shells.R {type=Audio}
 Shells.Volume {type=Float}
 Suppressor.Envelope.DecayTime {type=Time}
 Suppressor.L {type=Audio}
 Suppressor.MasterVolume {type=Float}
 Suppressor.Mix.BaseWeapon {type=Float}
 Suppressor.Mix.Suppressor {type=Float}
 Suppressor.R {type=Audio}
 Sweetener.L {type=Audio}
 Sweetener.R {type=Audio}
 Sweetener.Volume {type=Float}
 UE.Source.OnPlay {type=Trigger}
 Weapon.FireRate {type=Float}
 Weapon.IsSuppressorEnabled {type=Bool}
 Worldization.ActivationThreshold {type=Float}
 Worldization.EnvelopeMaxDecayTime {type=Float}
 Worldization.EnvelopeMinDecayTime {type=Float}
]
outputs [
 Out.C {type=Audio}
 Out.L {type=Audio}
 Out.LFE {type=Audio}
 Out.Ls {type=Audio}
 Out.R {type=Audio}
 Out.Rs {type=Audio}
]
pages [
 Default {
  is_build_graph=true
  input_defaults {
   LowAmmo.C=null
   LowAmmo.Volume=1
   MainFire.C=null
   MainFire.L=null
   MainFire.LFE=null
   MainFire.Ls=null
   MainFire.R=null
   MainFire.Rs=null
   MainFire.Volume=1
   Mech.C=null
   Mech.L=null
   Mech.R=null
   Mech.Volume=1
   OnGunshot=null
   OnStopFire.L=null
   OnStopFire.R=null
   OnStopFire.Volume=1
   Shells.L=null
   Shells.R=null
   Shells.Volume=1
   Suppressor.Envelope.DecayTime=0.195
   Suppressor.L=null
   Suppressor.MasterVolume=1
   Suppressor.Mix.BaseWeapon=0.3
   Suppressor.Mix.Suppressor=1
   Suppressor.R=null
   Sweetener.L=null
   Sweetener.R=null
   Sweetener.Volume=1
   UE.Source.OnPlay=null
   Weapon.FireRate=1
   Weapon.IsSuppressorEnabled=null
   Worldization.ActivationThreshold=0.9
   Worldization.EnvelopeMaxDecayTime=2
   Worldization.EnvelopeMinDecayTime=0.9
  }
  nodes [
   n1 = Convert.Float:Bool(Bool=^Weapon.IsSuppressorEnabled)
   ADEnvelopeADEnvelope_1 = AD Envelope.AD Envelope:Float(Trigger=^OnGunshot, Decay Time=^Suppressor.Envelope.DecayTime) {also {defaults {Attack Curve=2, Decay Curve=0.5}}}
   n2 = Modulation.GetModulatorValue() {also {defaults {Modulator=/FortMetaSoundWeapon/Items/Modulation/CB_WeaponFire_1p_Mech}}}
   n3 = UE.Multiply:Float(PrimaryOperand=^Mech.Volume, AdditionalOperands=@n2.Out)
   AudioMixerAudioMixerMono3_0 = "AudioMixer.Audio Mixer (Mono, 3)"(
    In 2=^LowAmmo.C
    In 0=^MainFire.C
    Gain 0=^MainFire.Volume
    In 1=^Mech.C
    Gain 1=@n3.Out
    Gain 2=UE.Multiply:Float(
     PrimaryOperand=^LowAmmo.Volume
     AdditionalOperands=Modulation.GetModulatorValue() {also {defaults {Modulator=/FortMetaSoundWeapon/Items/Modulation/CB_WeaponFire_1p_LowAmmo}}}.Out
    ).Out
   )
   n4 = Modulation.GetModulatorValue() {also {defaults {Modulator=&AudioWorldization_Bus_Enclosure}}}
   ADEnvelopeADEnvelope_2 = AD Envelope.AD Envelope:Float(
    Trigger=^OnGunshot
    Decay Time=UE.ConversionFloatToTime(In=MapRange.MapRange:Float(Out Range A=^Worldization.EnvelopeMaxDecayTime, Out Range B=^Worldization.EnvelopeMinDecayTime, In=@n4.Out).Out Value).Out
   ) {also {defaults {Attack Time=0, Decay Curve=0.5, Hard Reset="<MetaSoundBuilderResult.SUCCEEDED: 0>"}}}
   CrossfadeTriggerRouteFloat2_0 = "Crossfade.Trigger Route (Float, 2):Float"(
    In 0=@ADEnvelopeADEnvelope_2.Out Envelope
    Crossfade Value=MapRange.MapRange:Float(In Range A=^Worldization.ActivationThreshold, In=@n4.Out) {also {defaults {Out Range A=1, Out Range B=0}}}.Out Value
   ) {also {defaults {In 1=1}}}
   ADEnvelopeADEnvelope_0 = AD Envelope.AD Envelope:Float(
    Trigger=^UE.Source.OnPlay
    Decay Time=UE.ConversionFloatToTime(In=UE.Multiply:Float(PrimaryOperand=^Weapon.FireRate) {also {defaults {AdditionalOperands=1.2}}}.Out).Out
   ) {also {defaults {Attack Time=0, Hard Reset="<MetaSoundBuilderResult.SUCCEEDED: 0>"}}}
   n5 = MapRange.MapRange:Float(In=@ADEnvelopeADEnvelope_0.Out Envelope) {also {defaults {Out Range A=1}}}
   n6 = VariableAccessor.Float(Variable=VariableMutator.Float(
     Variable=InitVariable.Float().Variable
     Value=UE.Multiply:Float(PrimaryOperand=@CrossfadeTriggerRouteFloat2_0.Out, AdditionalOperands=@n5.Out Value).Out
    ).Variable)
   n7 = UE.Multiply:Audio by Float(PrimaryOperand=@AudioMixerAudioMixerMono3_0.Out, AdditionalOperands=@n6.Value)
   n8 = UE.Multiply:Audio by Float(AdditionalOperands=@ADEnvelopeADEnvelope_1.Out Envelope, PrimaryOperand=@n7.Out)
   Multiply_0 = UE.Multiply:Audio by Float(
    AdditionalOperands=@n1.Float
    PrimaryOperand=UE.Multiply:Audio by Float(
     AdditionalOperands=^Suppressor.MasterVolume
     PrimaryOperand=UE.Multiply:Audio by Float(PrimaryOperand=@n8.Out) {also {defaults {AdditionalOperands=0.6}}}.Out
    ).Out
   )
   n9 = UE.Subtract:Float(AdditionalOperands=@n1.Float) {also {defaults {PrimaryOperand=0.6}}}
   Add_0 = UE.Add:Audio(PrimaryOperand=@Multiply_0.Out, AdditionalOperands=UE.Multiply:Audio by Float(PrimaryOperand=@n7.Out, AdditionalOperands=@n9.Out).Out)
   AudioMixerAudioMixerStereo4_0 = "AudioMixer.Audio Mixer (Stereo, 4)"(
    In 0 L=^MainFire.L
    In 0 R=^MainFire.R
    Gain 0=^MainFire.Volume
    In 1 L=^OnStopFire.L
    In 1 R=^OnStopFire.R
    In 3 L=^Shells.L
    In 3 R=^Shells.R
    In 2 L=^Sweetener.L
    In 2 R=^Sweetener.R
    Gain 1=UE.Multiply:Float(
     PrimaryOperand=^OnStopFire.Volume
     AdditionalOperands=Modulation.GetModulatorValue() {also {defaults {Modulator=/FortMetaSoundWeapon/Items/Modulation/CB_WeaponFire_1p_OnStop}}}.Out
    ).Out
    Gain 2=UE.Multiply:Float(
     PrimaryOperand=^Sweetener.Volume
     AdditionalOperands=Modulation.GetModulatorValue() {also {defaults {Modulator=/FortMetaSoundWeapon/Items/Modulation/CB_WeaponFire_1p_Sweetener}}}.Out
    ).Out
    Gain 3=UE.Multiply:Float(PrimaryOperand=^Shells.Volume, AdditionalOperands=@n2.Out).Out
   )
   AudioMixerAudioMixerStereo2_0 = "AudioMixer.Audio Mixer (Stereo, 2)"(In 1 L=^Mech.L, In 1 R=^Mech.R, In 0 L=@AudioMixerAudioMixerStereo4_0.Out L, In 0 R=@AudioMixerAudioMixerStereo4_0.Out R, Gain 1=@n3.Out)
   n10 = UE.Multiply:Audio by Float(PrimaryOperand=@AudioMixerAudioMixerStereo2_0.Out R, AdditionalOperands=@n6.Value)
   n11 = UE.Multiply:Audio by Float(PrimaryOperand=@AudioMixerAudioMixerStereo2_0.Out L, AdditionalOperands=@n6.Value)
   AudioMixerAudioMixerStereo2_1 = "AudioMixer.Audio Mixer (Stereo, 2)"(
    In 0 L=^Suppressor.L
    Gain 1=^Suppressor.Mix.BaseWeapon
    Gain 0=^Suppressor.Mix.Suppressor
    In 0 R=^Suppressor.R
    In 1 R=UE.Multiply:Audio by Float(AdditionalOperands=@ADEnvelopeADEnvelope_1.Out Envelope, PrimaryOperand=@n10.Out).Out
    In 1 L=UE.Multiply:Audio by Float(AdditionalOperands=@ADEnvelopeADEnvelope_1.Out Envelope, PrimaryOperand=@n11.Out).Out
   )
   Add_1 = UE.Add:Audio(
    PrimaryOperand=UE.Multiply:Audio by Float(
     AdditionalOperands=@n1.Float
     PrimaryOperand=UE.Multiply:Audio by Float(AdditionalOperands=^Suppressor.MasterVolume, PrimaryOperand=@AudioMixerAudioMixerStereo2_1.Out R).Out
    ).Out
    AdditionalOperands=UE.Multiply:Audio by Float(PrimaryOperand=@n10.Out, AdditionalOperands=@n9.Out).Out
   )
   Add_2 = UE.Add:Audio(
    AdditionalOperands=UE.Multiply:Audio by Float(PrimaryOperand=@n11.Out, AdditionalOperands=@n9.Out).Out
    PrimaryOperand=UE.Multiply:Audio by Float(
     AdditionalOperands=@n1.Float
     PrimaryOperand=UE.Multiply:Audio by Float(AdditionalOperands=^Suppressor.MasterVolume, PrimaryOperand=@AudioMixerAudioMixerStereo2_1.Out L).Out
    ).Out
   )
   InitVariable.Int32()
   InitVariable.Trigger()
   InitVariable.Trigger()
   Multiply_1 = UE.Multiply:Audio by Float(
    PrimaryOperand=UE.Multiply:Audio by Float(
     PrimaryOperand=UE.Multiply:Audio by Float(PrimaryOperand=^MainFire.Ls, AdditionalOperands=^MainFire.Volume).Out
     AdditionalOperands=@n6.Value
    ).Out
    AdditionalOperands=@n9.Out
   )
   Multiply_11 = UE.Multiply:Audio by Float(
    PrimaryOperand=UE.Multiply:Audio by Float(
     PrimaryOperand=UE.Multiply:Audio by Float(PrimaryOperand=^MainFire.Rs, AdditionalOperands=^MainFire.Volume).Out
     AdditionalOperands=@n6.Value
    ).Out
    AdditionalOperands=@n9.Out
   )
   Multiply_4 = UE.Multiply:Audio by Float(
    PrimaryOperand=UE.Multiply:Audio by Float(
     PrimaryOperand=UE.Multiply:Audio by Float(PrimaryOperand=^MainFire.LFE, AdditionalOperands=^MainFire.Volume).Out
     AdditionalOperands=@n6.Value
    ).Out
    AdditionalOperands=@n9.Out
   )
  ]
  links [
   {dst="@Out.C", src=Add_0.Out}
   {dst="@Out.R", src=Add_1.Out}
   {dst="@Out.L", src=Add_2.Out}
   {dst="@Out.Ls", src=Multiply_1.Out}
   {dst="@Out.Rs", src=Multiply_11.Out}
   {dst="@Out.LFE", src=Multiply_4.Out}
  ]
 }
 Mobile {
  nodes [
   InitVariable.Int32()
   InitVariable.Trigger()
   InitVariable.Trigger()
   AudioMixerAudioMixerStereo3_0 = "AudioMixer.Audio Mixer (Stereo, 3)"(
    In 2 L=^LowAmmo.C
    In 2 R=^LowAmmo.C
    In 0 L=^MainFire.L
    In 0 R=^MainFire.R
    Gain 0=^MainFire.Volume
    In 1 L=^Mech.C
    In 1 R=^Mech.C
    Gain 1=UE.Multiply:Float(
     PrimaryOperand=^Mech.Volume
     AdditionalOperands=Modulation.GetModulatorValue() {also {defaults {Modulator=&CB_WeaponFire_1p_Mech}}}.Out
    ).Out
    Gain 2=UE.Multiply:Float(
     PrimaryOperand=^LowAmmo.Volume
     AdditionalOperands=Modulation.GetModulatorValue() {also {defaults {Modulator=&CB_WeaponFire_1p_LowAmmo}}}.Out
    ).Out
   )
   n1 = Convert.Float:Bool()
   ADEnvelopeADEnvelope_0 = AD Envelope.AD Envelope:Float(
    Trigger=^UE.Source.OnPlay
    Decay Time=UE.ConversionFloatToTime(In=UE.Multiply:Float(PrimaryOperand=^Weapon.FireRate) {also {defaults {AdditionalOperands=1.2}}}.Out).Out
   ) {also {defaults {Attack Time=0, Hard Reset="<MetaSoundBuilderResult.SUCCEEDED: 0>"}}}
   Multiply_4 = UE.Multiply:Float(
    AdditionalOperands=@n1.Float
    PrimaryOperand=MapRange.MapRange:Float(In=@ADEnvelopeADEnvelope_0.Out Envelope) {also {defaults {Out Range A=0.9}}}.Out Value
   )
   Add_0 = UE.Add:Float(PrimaryOperand=@Multiply_4.Out, AdditionalOperands=UE.Subtract:Float(AdditionalOperands=@n1.Float) {also {defaults {PrimaryOperand=1}}}.Out)
   n2 = Modulation.GetModulatorValue() {also {defaults {Modulator=&AudioWorldization_Bus_Enclosure}}}
   ADEnvelopeADEnvelope_1 = AD Envelope.AD Envelope:Float(
    Trigger=^OnGunshot
    Decay Time=UE.ConversionFloatToTime(In=MapRange.MapRange:Float(In=@n2.Out) {also {defaults {Out Range A=2, Out Range B=0.9}}}.Out Value).Out
   ) {also {defaults {Attack Time=0, Decay Curve=0.5, Hard Reset="<MetaSoundBuilderResult.SUCCEEDED: 0>"}}}
   CrossfadeTriggerRouteFloat2_0 = "Crossfade.Trigger Route (Float, 2):Float"(
    In 0=@ADEnvelopeADEnvelope_1.Out Envelope
    Crossfade Value=MapRange.MapRange:Float(In=@n2.Out) {also {defaults {In Range A=0.9, Out Range A=1, Out Range B=0}}}.Out Value
   ) {also {defaults {In 1=1}}}
   n3 = VariableAccessor.Float(Variable=VariableMutator.Float(
     Variable=InitVariable.Float().Variable
     Value=UE.Multiply:Float(AdditionalOperands=@Add_0.Out, PrimaryOperand=@CrossfadeTriggerRouteFloat2_0.Out).Out
    ).Variable)
   Multiply_0 = UE.Multiply:Audio by Float(PrimaryOperand=@AudioMixerAudioMixerStereo3_0.Out L, AdditionalOperands=@n3.Value)
   Multiply_1 = UE.Multiply:Audio by Float(PrimaryOperand=@AudioMixerAudioMixerStereo3_0.Out R, AdditionalOperands=@n3.Value)
  ]
  links [{dst="@Out.L", src=Multiply_0.Out}, {dst="@Out.R", src=Multiply_1.Out}]
 }
]
