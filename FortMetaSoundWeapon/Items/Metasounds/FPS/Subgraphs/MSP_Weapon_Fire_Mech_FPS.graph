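# MetaSound patch for the mechanical layer of first-person weapon fire, playing a random low-repeat sound from Mech.Sounds per gunshot with randomized pitch and volume.
# A trigger sequence alternates shots between two stereo wave players so tails can overlap, with bUseSingleVoice controlling whether only one is used.
# Each player's gain applies an aim-down-sights multiplier before mixing.
# The Mobile page uses one mono player, omits aiming logic and feeds only the left output.
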MetaSoundPatch /FortMetaSoundWeapon/Items/Metasounds/FPS/Subgraphs/MSP_Weapon_Fire_Mech_FPS
inputs [
 bUseSingleVoice {type=Bool}
 Mech.ADS.VolumeMultiply {type=Float}
 Mech.Random.Pitch.Max {type=Float}
 Mech.Random.Pitch.Min {type=Float}
 Mech.Random.Volume.Max {type=Float}
 Mech.Random.Volume.Min {type=Float}
 Mech.Sounds {type=WaveAsset:Array}
 OnGunshot {type=Trigger}
 Weapon.IsAimingDownSights {type=Bool}
]
outputs [
 Mech.L {type=Audio}
 Mech.R {type=Audio}
]
pages [
 Default {
  is_build_graph=true
  input_defaults {
   Mech.ADS.VolumeMultiply=1
   Mech.Random.Pitch.Max=1.05
   Mech.Random.Pitch.Min=0.95
   Mech.Random.Volume.Max=1.05
   Mech.Random.Volume.Min=0.95
   OnGunshot=null
   Weapon.IsAimingDownSights=null
   bUseSingleVoice=null
   Mech.Sounds []
  }
  nodes [
   n1 = Convert.Float:Bool(Bool=^Weapon.IsAimingDownSights)
   n2 = UE.Multiply:Float(PrimaryOperand=^Mech.ADS.VolumeMultiply, AdditionalOperands=@n1.Float)
   n3 = UE.Subtract:Float(AdditionalOperands=@n1.Float) {also {defaults {PrimaryOperand=1}}}
   n4 = TriggerCompare.Clamp:Bool(Compare=^OnGunshot, A=^bUseSingleVoice) {also {defaults {B="<MetaSoundBuilderResult.SUCCEEDED: 0>"}}}
   TriggerSequence2_0 = "UE.Trigger Sequence (2)"(In=@n4.False)
   n5 = VariableAccessor.Trigger(Variable=VariableMutator.Trigger(Variable=InitVariable.Trigger().Variable, Value=@TriggerSequence2_0.Out 1).Variable)
   n6 = VariableAccessor.Trigger(Variable=@n5.Variable)
   n7 = Convert.Float:Bool(Bool=^Weapon.IsAimingDownSights)
   Multiply_0 = UE.Multiply:Float(PrimaryOperand=^Mech.ADS.VolumeMultiply, AdditionalOperands=@n7.Float)
   n8 = UE.Subtract:Float(AdditionalOperands=@n7.Float) {also {defaults {PrimaryOperand=1}}}
   TriggerAnyTriggerAny2_0 = "TriggerAny.Trigger Any (2)"(In 0=@n4.True, In 1=@TriggerSequence2_0.Out 0)
   n9 = VariableAccessor.Trigger(Variable=VariableMutator.Trigger(Variable=InitVariable.Trigger().Variable, Value=@TriggerAnyTriggerAny2_0.Out).Variable)
   n10 = VariableAccessor.Trigger(Variable=@n9.Variable)
   n11 = UE.RandomFloat(Max=^Mech.Random.Pitch.Max, Min=^Mech.Random.Pitch.Min, Next=^OnGunshot)
   FrequencyMultipliertoSemitone_1 = UE.Frequency Multiplier to Semitone:Float(Frequency Multiplier=@n11.Value)
   n12 = Array.Num:WaveAsset:Array(Array=^Mech.Sounds)
   ArrayRandomGet_0 = Array.Random Get:WaveAsset:Array(
    In Array=^Mech.Sounds
    Next=^OnGunshot
    No Repeats=Max.Max:Int32(A=UE.Subtract:Int32(PrimaryOperand=@n12.Num) {also {defaults {AdditionalOperands=2}}}.Out) {also {defaults {B=1}}}.Value
   ) {also {defaults {Enable Shared State="<MetaSoundBuilderResult.SUCCEEDED: 0>", Weights []}}}
   WavePlayer_0 = UE.Wave Player:Stereo(Pitch Shift=@FrequencyMultipliertoSemitone_1.Semitones, Wave Asset=UE.Reroute(Value=@ArrayRandomGet_0.Value).Value, Play=@n5.Value)
   FrequencyMultipliertoSemitone_0 = UE.Frequency Multiplier to Semitone:Float(Frequency Multiplier=@n11.Value)
   WavePlayer_1 = UE.Wave Player:Stereo(Wave Asset=@ArrayRandomGet_0.Value, Pitch Shift=@FrequencyMultipliertoSemitone_0.Semitones, Play=@n9.Value)
   AudioMixerAudioMixerStereo2_0 = "AudioMixer.Audio Mixer (Stereo, 2)"(
    Gain 1=UE.Multiply:Float(
     AdditionalOperands=UE.Add:Float(PrimaryOperand=@n2.Out, AdditionalOperands=@n3.Out).Out
     PrimaryOperand=UE.RandomFloat(Max=^Mech.Random.Volume.Max, Min=^Mech.Random.Volume.Min, Next=@n6.Value).Value
    ).Out
    Gain 0=UE.Multiply:Float(
     AdditionalOperands=UE.Add:Float(PrimaryOperand=@Multiply_0.Out, AdditionalOperands=@n8.Out).Out
     PrimaryOperand=UE.RandomFloat(Max=^Mech.Random.Volume.Max, Min=^Mech.Random.Volume.Min, Next=@n10.Value).Value
    ).Out
    In 1 L=@WavePlayer_0.Out Left
    In 1 R=@WavePlayer_0.Out Right
    In 0 L=@WavePlayer_1.Out Left
    In 0 R=@WavePlayer_1.Out Right
   )
  ]
  links [{dst="@Mech.L", src=AudioMixerAudioMixerStereo2_0.Out L}, {dst="@Mech.R", src=AudioMixerAudioMixerStereo2_0.Out R}]
 }
 Mobile {
  nodes [
   InitVariable.Trigger()
   InitVariable.Trigger()
   n1 = Array.Num:WaveAsset:Array(Array=^Mech.Sounds)
   ArrayRandomGet_0 = Array.Random Get:WaveAsset:Array(
    In Array=^Mech.Sounds
    Next=^OnGunshot
    No Repeats=Max.Max:Int32(A=UE.Subtract:Int32(PrimaryOperand=@n1.Num) {also {defaults {AdditionalOperands=2}}}.Out) {also {defaults {B=1}}}.Value
   ) {also {defaults {Enable Shared State="<MetaSoundBuilderResult.SUCCEEDED: 0>", Weights []}}}
   FrequencyMultipliertoSemitone_0 = UE.Frequency Multiplier to Semitone:Float(Frequency Multiplier=UE.RandomFloat(Max=^Mech.Random.Pitch.Max, Min=^Mech.Random.Pitch.Min, Next=^OnGunshot).Value)
   WavePlayer_0 = UE.Wave Player:Mono(Play=@ArrayRandomGet_0.On Next, Wave Asset=@ArrayRandomGet_0.Value, Pitch Shift=@FrequencyMultipliertoSemitone_0.Semitones)
   Multiply_0 = UE.Multiply:Audio by Float(
    AdditionalOperands=UE.RandomFloat(Max=^Mech.Random.Volume.Max, Min=^Mech.Random.Volume.Min, Next=UE.Reroute(Value=@ArrayRandomGet_0.On Next).Value).Value
    PrimaryOperand=@WavePlayer_0.Out Mono
   )
  ]
  links [{dst="@Mech.L", src=Multiply_0.Out}]
 }
]
