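# MetaSound patch that plays a randomized shell-eject sound on each weapon gunshot.
# A random wave, with a no-repeat count scaled to array size, plays with random pitch and volume.
# The default page adds a trigger delay, aim-down-sights volume scaling and stereo panning into Mech.L and Mech.R.
# The Mobile page omits all three and sends mono audio to Mech.L only.
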MetaSoundPatch /FortMetaSoundWeapon/Items/Metasounds/FPS/Subgraphs/MSP_Weapon_Fire_Shell_Eject
inputs [
 Delay Time {type=Time}
 OnGunshot {type=Trigger}
 Pan Amount {type=Float}
 Shells.ADS.VolumeMultiply {type=Float}
 Shells.Random.Pitch.Max {type=Float}
 Shells.Random.Pitch.Min {type=Float}
 Shells.Random.Volume.Max {type=Float}
 Shells.Random.Volume.Min {type=Float}
 Shells.Sounds {type=WaveAsset:Array}
 Weapon.IsAimingDownSights {type=Bool}
]
outputs [
 Mech.L {type=Audio}
 Mech.R {type=Audio}
]
pages [
 Default {
  is_build_graph=true
  input_defaults {
   Delay Time=0
   OnGunshot=null
   Pan Amount=0
   Shells.ADS.VolumeMultiply=1
   Shells.Random.Pitch.Max=1.05
   Shells.Random.Pitch.Min=0.95
   Shells.Random.Volume.Max=1.05
   Shells.Random.Volume.Min=0.95
   Weapon.IsAimingDownSights=null
   Shells.Sounds []
  }
  nodes [
   n1 = Convert.Float:Bool(Bool=^Weapon.IsAimingDownSights)
   n2 = UE.Multiply:Float(PrimaryOperand=^Shells.ADS.VolumeMultiply, AdditionalOperands=@n1.Float)
   n3 = UE.Subtract:Float(AdditionalOperands=@n1.Float) {also {defaults {PrimaryOperand=1}}}
   n4 = Array.Num:WaveAsset:Array(Array=^Shells.Sounds)
   ArrayRandomGet_0 = Array.Random Get:WaveAsset:Array(
    Next=^OnGunshot
    In Array=^Shells.Sounds
    No Repeats=Max.Max:Int32(A=UE.Subtract:Int32(PrimaryOperand=@n4.Num) {also {defaults {AdditionalOperands=2}}}.Out) {also {defaults {B=1}}}.Value
   ) {also {defaults {Enable Shared State="<MetaSoundBuilderResult.SUCCEEDED: 0>", Weights []}}}
   TriggerDelay_0 = UE.Trigger Delay(Delay Time=^Delay Time, In=@ArrayRandomGet_0.On Next)
   n5 = UE.RandomFloat(Max=^Shells.Random.Volume.Max, Min=^Shells.Random.Volume.Min, Next=@TriggerDelay_0.Out)
   FrequencyMultipliertoSemitone_0 = UE.Frequency Multiplier to Semitone:Float(Frequency Multiplier=UE.RandomFloat(Next=^OnGunshot, Max=^Shells.Random.Pitch.Max, Min=^Shells.Random.Pitch.Min).Value)
   WavePlayer_0 = UE.Wave Player:Mono(Wave Asset=@ArrayRandomGet_0.Value, Pitch Shift=@FrequencyMultipliertoSemitone_0.Semitones, Play=@n5.On Next)
   Multiply_0 = UE.Multiply:Audio by Float(
    AdditionalOperands=UE.Multiply:Float(AdditionalOperands=UE.Add:Float(PrimaryOperand=@n2.Out, AdditionalOperands=@n3.Out).Out, PrimaryOperand=@n5.Value).Out
    PrimaryOperand=@WavePlayer_0.Out Mono
   )
   StereoPanner_0 = UE.Stereo Panner(Pan Amount=^Pan Amount, In=@Multiply_0.Out)
  ]
  links [{dst="@Mech.L", src=StereoPanner_0.Out Left}, {dst="@Mech.R", src=StereoPanner_0.Out Right}]
 }
 Mobile {
  nodes [
   InitVariable.Trigger()
   InitVariable.Trigger()
   n1 = Array.Num:WaveAsset:Array(Array=^Shells.Sounds)
   ArrayRandomGet_0 = Array.Random Get:WaveAsset:Array(
    Next=^OnGunshot
    In Array=^Shells.Sounds
    No Repeats=Max.Max:Int32(A=UE.Subtract:Int32(PrimaryOperand=@n1.Num) {also {defaults {AdditionalOperands=2}}}.Out) {also {defaults {B=1}}}.Value
   ) {also {defaults {Enable Shared State="<MetaSoundBuilderResult.SUCCEEDED: 0>", Weights []}}}
   FrequencyMultipliertoSemitone_0 = UE.Frequency Multiplier to Semitone:Float(Frequency Multiplier=UE.RandomFloat(Next=^OnGunshot, Max=^Shells.Random.Pitch.Max, Min=^Shells.Random.Pitch.Min).Value)
   WavePlayer_0 = UE.Wave Player:Mono(Play=@ArrayRandomGet_0.On Next, Wave Asset=@ArrayRandomGet_0.Value, Pitch Shift=@FrequencyMultipliertoSemitone_0.Semitones)
   Multiply_0 = UE.Multiply:Audio by Float(
    AdditionalOperands=UE.RandomFloat(Max=^Shells.Random.Volume.Max, Min=^Shells.Random.Volume.Min, Next=UE.Reroute(Value=@ArrayRandomGet_0.On Next).Value).Value
    PrimaryOperand=@WavePlayer_0.Out Mono
   )
  ]
  links [{dst="@Mech.L", src=Multiply_0.Out}]
 }
]
