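# MetaSound patch that selects between two float inputs based on a boolean, evaluated when TriggerCompare fires.
# A Trigger Compare node tests the bool, and its True and False outputs drive Set and Reset on a Value node.
# Set outputs FloatIfTrue and Reset returns to FloatIfFalse, the initial value.
# The node's On Set and On Reset pass out as TrueTrigger and FalseTrigger alongside FloatOut.
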MetaSoundPatch /FortMetasound/Patches/MSP_SetFloatWithBool
inputs [
 bBool {type=Bool}
 FloatIfFalse {type=Float}
 FloatIfTrue {type=Float}
 TriggerCompare {type=Trigger}
]
outputs [
 FalseTrigger {type=Trigger}
 FloatOut {type=Float}
 TrueTrigger {type=Trigger}
]
pages [Default {
  input_defaults {FloatIfFalse=1, FloatIfTrue=0, TriggerCompare=null, bBool=null}
  nodes [
   n1 = TriggerCompare.Clamp:Bool(Compare=^TriggerCompare, A=^bBool) {also {defaults {B="<MetaSoundBuilderResult.SUCCEEDED: 0>"}}}
   ValueValue_0 = Value.Value:Float(Init Value=^FloatIfFalse, Target Value=^FloatIfTrue, Reset=@n1.False, Set=@n1.True)
  ]
  links [
   {dst="@FalseTrigger", src=ValueValue_0.On Reset}
   {dst="@TrueTrigger", src=ValueValue_0.On Set}
   {dst="@FloatOut", src=ValueValue_0.Output Value}
  ]
 }]
