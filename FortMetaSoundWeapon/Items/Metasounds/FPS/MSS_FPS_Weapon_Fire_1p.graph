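# One-shot 5.1 MetaSound source for first-person weapon fire.
# A fire-control patch turns single, burst and firing inputs into gunshot triggers for main, mech, low-ammo, shell, suppressor and sweetener layer patches.
# Aim-down-sights drives waveshaping, filter gain and LFE level around the FPS mixing patch, whose L/R sum also feeds the AB_WeaponFire bus.
# Main and low-ammo completion gates a delayed finish; the Mobile page omits shells and sweetener.
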MetaSoundSource /FortMetaSoundWeapon/Items/Metasounds/FPS/MSS_FPS_Weapon_Fire_1p
inputs [
 bIsFiring {type=Bool}
 BurstRate {type=Time}
 bUseSingleVoice {type=Bool}
 FireBurst {type=Trigger}
 FireSingleShot {type=Trigger}
 FPSMix.7 Weapon Output Volume {type=Float}
 FPSMIX.ADS.Filter.Bandwidth {type=Float}
 FPSMIX.ADS.Filter.CutoffFreq {type=Float}
 FPSMIX.ADS.Filter.GainADS {type=Float}
 FPSMIX.ADS.Filter.GainHipFire {type=Float}
 FPSMIX.ADS.Filter.Wave Shaper Main Fire Sub ADS Fire Amount {type=Float}
 FPSMIX.ADS.Filter.Wave Shaper Mech ADS Fire Amount {type=Float}
 FPSMix.ADS.LFEVolume {type=Float}
 FPSMix.CenterVolume {type=Float}
 FPSMix.LeftSurroundVolume {type=Float}
 FPSMix.LeftVolume {type=Float}
 FPSMix.LFEVolume {type=Float}
 FPSMix.RightSurroundVolume {type=Float}
 FPSMix.RightVolume {type=Float}
 FPSMix.Worldization.ActivationThreshold {type=Float}
 FPSMix.Worldization.EnvelopeMaxDecayTime {type=Float}
 FPSMix.Worldization.EnvelopeMinDecayTime {type=Float}
 LowAmmo.ADS.VolumeMultiply {type=Float}
 LowAmmo.Random.Pitch.Max {type=Float}
 LowAmmo.Random.Pitch.Min {type=Float}
 LowAmmo.Random.Volume.Max {type=Float}
 LowAmmo.Random.Volume.Min {type=Float}
 LowAmmo.Sounds {type=WaveAsset:Array}
 LowAmmo.Threshold {type=Int32}
 LowAmmo.Volume {type=Float}
 MainFire.ADS.VolumeMultiply {type=Float}
 MainFire.Random.Pitch.Max {type=Float}
 MainFire.Random.Pitch.Min {type=Float}
 MainFire.Random.Volume.Max {type=Float}
 MainFire.Random.Volume.Min {type=Float}
 MainFire.Sounds {type=WaveAsset:Array}
 MainFire.Volume {type=Float}
 Mech.ADSVolumeMultiply {type=Float}
 Mech.Random.Pitch.Max {type=Float}
 Mech.Random.Pitch.Min {type=Float}
 Mech.Random.Volume.Max {type=Float}
 Mech.Random.Volume.Min {type=Float}
 Mech.Sounds {type=WaveAsset:Array}
 Mech.Volume {type=Float}
 OnStop {type=Trigger}
 OnStopFire.Sounds {type=WaveAsset:Array}
 OnStopFire.Volume {type=Float}
 RetriggerDelayTime {type=Time}
 RoundsPerBurst {type=Int32}
 Shells.ADS.VolumeMultiply {type=Float}
 Shells.DelayTime {type=Time}
 Shells.PanAmount {type=Float}
 Shells.Random.Pitch.Max {type=Float}
 Shells.Random.Pitch.Min {type=Float}
 Shells.Random.Volume.Max {type=Float}
 Shells.Random.Volume.Min {type=Float}
 Shells.Sounds {type=WaveAsset:Array}
 Shells.Volume {type=Float}
 StopBurst {type=Trigger}
 Suppressor.BaseWeaponMix {type=Float}
 Suppressor.EnvelopeDecayTime {type=Time}
 Suppressor.MasterVolume {type=Float}
 Suppressor.Random.Pitch.Max {type=Float}
 Suppressor.Random.Pitch.Min {type=Float}
 Suppressor.Random.Volume.Max {type=Float}
 Suppressor.Random.Volume.Min {type=Float}
 Suppressor.Sounds {type=WaveAsset:Array}
 Suppressor.SuppressorVolume {type=Float}
 Sweetener.ADS.VolumeMultiply {type=Float}
 Sweetener.Random.Pitch.Max {type=Float}
 Sweetener.Random.Pitch.Min {type=Float}
 Sweetener.Random.Volume.Max {type=Float}
 Sweetener.Random.Volume.Min {type=Float}
 Sweetener.Sounds {type=WaveAsset:Array}
 Sweetener.Volume {type=Float}
 UE.Source.OnPlay {type=Trigger}
 Weapon.AmmoRemaining {type=Int32}
 Weapon.ClipSize {type=Int32}
 Weapon.FireRate {type=Float}
 Weapon.FireStartTime {type=Float}
 Weapon.IsAimingDownSights {type=Bool}
 Weapon.IsSuppressorEnabled {type=Bool}
]
interfaces [Fort.Weapon:1.0, UE.OutputFormat.5dot1:1.0, UE.Source.OneShot:1.0, UE.Source:1.1]
outputs [
 bOnAlmostFinished {type=Bool}
 OnGunshotOut {type=Trigger}
 "UE.OutputFormat.5dot1.Audio:0" {type=Audio}
 "UE.OutputFormat.5dot1.Audio:1" {type=Audio}
 "UE.OutputFormat.5dot1.Audio:2" {type=Audio}
 "UE.OutputFormat.5dot1.Audio:3" {type=Audio}
 "UE.OutputFormat.5dot1.Audio:4" {type=Audio}
 "UE.OutputFormat.5dot1.Audio:5" {type=Audio}
 UE.Source.OneShot.OnFinished {type=Trigger}
]
pages [
 Default {
  is_build_graph=true
  input_defaults {
   BurstRate=0.4
   FPSMIX.ADS.Filter.Bandwidth=1
   FPSMIX.ADS.Filter.CutoffFreq=20000
   FPSMIX.ADS.Filter.GainADS=1
   FPSMIX.ADS.Filter.GainHipFire=1
   FPSMIX.ADS.Filter.Wave Shaper Main Fire Sub ADS Fire Amount=0
   FPSMIX.ADS.Filter.Wave Shaper Mech ADS Fire Amount=0
   FPSMix.7 Weapon Output Volume=0.75
   FPSMix.ADS.LFEVolume=1.5
   FPSMix.CenterVolume=1
   FPSMix.LFEVolume=1
   FPSMix.LeftSurroundVolume=1
   FPSMix.LeftVolume=1
   FPSMix.RightSurroundVolume=1
   FPSMix.RightVolume=1
   FPSMix.Worldization.ActivationThreshold=0.8
   FPSMix.Worldization.EnvelopeMaxDecayTime=4
   FPSMix.Worldization.EnvelopeMinDecayTime=2
   FireBurst=null
   FireSingleShot=null
   LowAmmo.ADS.VolumeMultiply=1
   LowAmmo.Random.Pitch.Max=1.05
   LowAmmo.Random.Pitch.Min=0.95
   LowAmmo.Random.Volume.Max=1.05
   LowAmmo.Random.Volume.Min=0.95
   LowAmmo.Threshold=10
   LowAmmo.Volume=1
   MainFire.ADS.VolumeMultiply=1
   MainFire.Random.Pitch.Max=1
   MainFire.Random.Pitch.Min=1
   MainFire.Random.Volume.Max=1.05
   MainFire.Random.Volume.Min=0.95
   MainFire.Volume=1
   Mech.ADSVolumeMultiply=1
   Mech.Random.Pitch.Max=1.05
   Mech.Random.Pitch.Min=0.95
   Mech.Random.Volume.Max=1.05
   Mech.Random.Volume.Min=0.95
   Mech.Volume=1
   OnStop=null
   OnStopFire.Volume=1
   RetriggerDelayTime=2
   RoundsPerBurst=2
   Shells.ADS.VolumeMultiply=1
   Shells.DelayTime=0
   Shells.PanAmount=0.5
   Shells.Random.Pitch.Max=1
   Shells.Random.Pitch.Min=1
   Shells.Random.Volume.Max=1.05
   Shells.Random.Volume.Min=0.95
   Shells.Volume=0.55
   StopBurst=null
   Suppressor.BaseWeaponMix=0.9
   Suppressor.EnvelopeDecayTime=0.195
   Suppressor.MasterVolume=1
   Suppressor.Random.Pitch.Max=1.05
   Suppressor.Random.Pitch.Min=0.95
   Suppressor.Random.Volume.Max=1.05
   Suppressor.Random.Volume.Min=0.95
   Suppressor.SuppressorVolume=1
   Sweetener.ADS.VolumeMultiply=1
   Sweetener.Random.Pitch.Max=1.05
   Sweetener.Random.Pitch.Min=0.95
   Sweetener.Random.Volume.Max=1.05
   Sweetener.Random.Volume.Min=0.95
   Sweetener.Volume=1
   UE.Source.OnPlay=null
   Weapon.AmmoRemaining=30
   Weapon.ClipSize=30
   Weapon.FireRate=0.496
   Weapon.FireStartTime=0
   Weapon.IsAimingDownSights=null
   Weapon.IsSuppressorEnabled=null
   bIsFiring=null
   bUseSingleVoice=null
   LowAmmo.Sounds []
   MainFire.Sounds []
   Mech.Sounds []
   OnStopFire.Sounds []
   Shells.Sounds []
   Suppressor.Sounds []
   Sweetener.Sounds []
  }
  nodes [
   MSP_Weapon_Fire_Control_1p_0 = /FortMetaSoundWeapon/Items/Metasounds/Subgraphs/MSP_Weapon_Fire_Control_1p(
    FireBurst=^FireBurst
    FireSingleShot=^FireSingleShot
    RoundsPerBurst=^RoundsPerBurst
    StopBurst=^StopBurst
    OnPlay=^UE.Source.OnPlay
    Weapon.AmmoRemaining=^Weapon.AmmoRemaining
    Weapon.FireRate=^Weapon.FireRate
    bIsFiring=^bIsFiring
    BurstRate=UE.ConversionTimeToFloat(In=^BurstRate).Out
   )
   n1 = VariableAccessor.Trigger(Variable=VariableMutator.Trigger(Variable=InitVariable.Trigger().Variable, Value=@MSP_Weapon_Fire_Control_1p_0.OutGunshot).Variable)
   n2 = /FortMetaSoundWeapon/Items/Metasounds/Subgraphs/MSP_Weapon_Fire_MainFire(
    MainFire.ADS.VolumeMultiply=^MainFire.ADS.VolumeMultiply
    MainFire.Random.Pitch.Min=^MainFire.Random.Pitch.Max
    MainFire.Random.Pitch.Max=^MainFire.Random.Pitch.Min
    MainFire.Random.Volume.Max=^MainFire.Random.Volume.Max
    MainFire.Random.Volume.Min=^MainFire.Random.Volume.Min
    MainFire.Sounds=^MainFire.Sounds
    Weapon.IsAimingDownSights=^Weapon.IsAimingDownSights
    bUseSingleVoice=^bUseSingleVoice
    OnGunshot=@n1.Value
   )
   TriggerOnValueChange_1 = UE.Trigger On Value Change:Bool()
   n3 = VariableAccessor.Trigger(Variable=@n1.Variable)
   n4 = VariableAccessor.Trigger(Variable=@n3.Variable)
   n5 = VariableAccessor.Trigger(Variable=@n4.Variable)
   n6 = VariableAccessor.Trigger(Variable=@n5.Variable)
   n7 = VariableAccessor.Trigger(Variable=@n6.Variable)
   n8 = VariableAccessor.Trigger(Variable=@n7.Variable)
   n9 = VariableAccessor.Trigger(Variable=@n8.Variable)
   n10 = VariableAccessor.Trigger(Variable=@n9.Variable)
   n11 = VariableAccessor.Trigger(Variable=@n10.Variable)
   TriggerAnyTriggerAny2_1 = "TriggerAny.Trigger Any (2)"(In 1=@TriggerOnValueChange_1.Trigger, In 0=@n11.Value)
   AudioMixerAudioMixerMono2_0 = "AudioMixer.Audio Mixer (Mono, 2)"(
    In 0=UE.Multiply:Audio by Float(AdditionalOperands=^FPSMix.CenterVolume, PrimaryOperand=@n2.MainFire.C).Out
    In 1=UE.WaveShaper:Audio(
     Amount=/FortMetasound/Patches/MSP_SetFloatWithBool(
      FloatIfTrue=^FPSMIX.ADS.Filter.Wave Shaper Main Fire Sub ADS Fire Amount
      bBool=^Weapon.IsAimingDownSights
      TriggerCompare=@TriggerAnyTriggerAny2_1.Out
     ).FloatOut
     In=@n2.MainFire.LFE
    ) {also {defaults {Type=4}}}.Out
   ) {also {defaults {Gain 1=0.6}}}
   TriggerOnValueChange_2 = UE.Trigger On Value Change:Bool(Value=^Weapon.IsAimingDownSights)
   n12 = VariableAccessor.Trigger(Variable=@n11.Variable)
   TriggerAnyTriggerAny2_3 = "TriggerAny.Trigger Any (2)"(In 1=@TriggerOnValueChange_2.Trigger, In 0=@n12.Value)
   TriggerOnValueChange_3 = UE.Trigger On Value Change:Bool(Value=^Weapon.IsAimingDownSights)
   TriggerAnyTriggerAny2_0 = "TriggerAny.Trigger Any (2)"(In 1=@TriggerOnValueChange_3.Trigger, In 0=@n10.Value)
   n13 = &MSP_SetFloatWithBool(
    FloatIfTrue=^FPSMIX.ADS.Filter.Wave Shaper Mech ADS Fire Amount
    bBool=^Weapon.IsAimingDownSights
    TriggerCompare=@TriggerAnyTriggerAny2_0.Out
   )
   n14 = /FortMetaSoundWeapon/Items/Metasounds/FPS/Subgraphs/MSP_Weapon_Fire_Mech_FPS(
    Mech.ADS.VolumeMultiply=^Mech.ADSVolumeMultiply
    Mech.Random.Pitch.Max=^Mech.Random.Pitch.Max
    Mech.Random.Pitch.Min=^Mech.Random.Pitch.Min
    Mech.Random.Volume.Max=^Mech.Random.Volume.Max
    Mech.Random.Volume.Min=^Mech.Random.Volume.Min
    Mech.Sounds=^Mech.Sounds
    Weapon.IsAimingDownSights=^Weapon.IsAimingDownSights
    bUseSingleVoice=^bUseSingleVoice
    OnGunshot=@n5.Value
   )
   AudioMixerAudioMixerMono2_2 = "AudioMixer.Audio Mixer (Mono, 2)"(
    Gain 0=UE.InterpTo:Audio(Target=&MSP_SetFloatWithBool(bBool=^Weapon.IsAimingDownSights, TriggerCompare=@TriggerAnyTriggerAny2_3.Out) {also {defaults {FloatIfFalse=0.75, FloatIfTrue=1}}}.FloatOut).Value
    In 0=UE.WaveShaper:Audio(Amount=@n13.FloatOut, In=@n14.Mech.L) {also {defaults {Type=4}}}.Out
   ) {also {defaults {Gain 1=0}}}
   n15 = VariableAccessor.Trigger(Variable=@n12.Variable)
   n16 = VariableAccessor.Trigger(Variable=@n15.Variable)
   LinearGaintoDecibels_0 = UE.Linear Gain to Decibels:Float(Linear Gain=&MSP_SetFloatWithBool(
     FloatIfTrue=^FPSMIX.ADS.Filter.GainADS
     FloatIfFalse=^FPSMIX.ADS.Filter.GainHipFire
     bBool=^Weapon.IsAimingDownSights
     TriggerCompare=@n16.Value
    ).FloatOut)
   BiquadFilter_0 = UE.Biquad Filter:Audio(
    Bandwidth=^FPSMIX.ADS.Filter.Bandwidth
    Cutoff Frequency=^FPSMIX.ADS.Filter.CutoffFreq
    Gain=@LinearGaintoDecibels_0.Decibels
    In=UE.Multiply:Audio by Float(AdditionalOperands=^FPSMix.LeftVolume, PrimaryOperand=@n2.MainFire.L).Out
   ) {also {defaults {Type=4}}}
   BiquadFilter_1 = UE.Biquad Filter:Audio(
    Bandwidth=^FPSMIX.ADS.Filter.Bandwidth
    Cutoff Frequency=^FPSMIX.ADS.Filter.CutoffFreq
    Gain=@LinearGaintoDecibels_0.Decibels
    In=UE.Multiply:Audio by Float(AdditionalOperands=^FPSMix.RightVolume, PrimaryOperand=@n2.MainFire.R).Out
   ) {also {defaults {Type=4}}}
   n17 = InitVariable.Int32()
   n18 = /FortMetaSoundWeapon/Items/Metasounds/Subgraphs/MSP_Weapon_Fire_LowAmmo(
    LowAmmo.ADS.VolumeMultiply=^LowAmmo.ADS.VolumeMultiply
    LowAmmo.Random.Pitch.Max=^LowAmmo.Random.Pitch.Max
    LowAmmo.Random.Pitch.Min=^LowAmmo.Random.Pitch.Min
    LowAmmo.Random.Volume.Max=^LowAmmo.Random.Volume.Max
    LowAmmo.Random.Volume.Min=^LowAmmo.Random.Volume.Min
    LowAmmo.Sounds=^LowAmmo.Sounds
    LowAmmo.Threshold=^LowAmmo.Threshold
    Weapon.IsAimingDownSights=^Weapon.IsAimingDownSights
    CurrentAmmo=VariableAccessor.Int32(Variable=VariableMutator.Int32(Variable=@n17.Variable, Value=@MSP_Weapon_Fire_Control_1p_0.CurrentAmmo).Variable).Value
    OnGunshot=@n6.Value
   )
   n19 = /FortMetaSoundWeapon/Items/Metasounds/FPS/Subgraphs/MSP_Weapon_Fire_Shell_Eject(
    Shells.ADS.VolumeMultiply=^Shells.ADS.VolumeMultiply
    Delay Time=^Shells.DelayTime
    Pan Amount=^Shells.PanAmount
    Shells.Random.Pitch.Max=^Shells.Random.Pitch.Max
    Shells.Random.Pitch.Min=^Shells.Random.Pitch.Min
    Shells.Random.Volume.Max=^Shells.Random.Volume.Max
    Shells.Random.Volume.Min=^Shells.Random.Volume.Min
    Shells.Sounds=^Shells.Sounds
    Weapon.IsAimingDownSights=^Weapon.IsAimingDownSights
    OnGunshot=@n15.Value
   )
   n20 = /FortMetaSoundWeapon/Items/Metasounds/Subgraphs/MSP_Weapon_Fire_Suppressor(
    Random.Pitch.Max=^Suppressor.Random.Pitch.Max
    Random.Pitch.Min=^Suppressor.Random.Pitch.Min
    Random.Volume.Max=^Suppressor.Random.Volume.Max
    Random.Volume.Min=^Suppressor.Random.Volume.Min
    Sounds=^Suppressor.Sounds
    Weapon.IsSuppressorEnabled=^Weapon.IsSuppressorEnabled
    bUseSingleVoice=^bUseSingleVoice
    OnGunshot=@n3.Value
   )
   n21 = /FortMetaSoundWeapon/Items/Metasounds/Subgraphs/MSP_Weapon_Fire_Sweetener(
    Sweetener.ADS.VolumeMultiply=^Sweetener.ADS.VolumeMultiply
    Sweetener.Random.Pitch.Max=^Sweetener.Random.Pitch.Max
    Sweetener.Random.Pitch.Min=^Sweetener.Random.Pitch.Min
    Sweetener.Random.Volume.Max=^Sweetener.Random.Volume.Max
    Sweetener.Random.Volume.Min=^Sweetener.Random.Volume.Min
    Sweetener.Sounds=^Sweetener.Sounds
    Weapon.IsAimingDownSights=^Weapon.IsAimingDownSights
    bUseSingleVoice=^bUseSingleVoice
    OnGunshot=@n7.Value
   )
   n22 = /FortMetaSoundWeapon/Items/Metasounds/FPS/Subgraphs/MSP_Weapon_Fire_FPS_Mixing_1P(
    Worldization.ActivationThreshold=^FPSMix.Worldization.ActivationThreshold
    Worldization.EnvelopeMaxDecayTime=^FPSMix.Worldization.EnvelopeMaxDecayTime
    Worldization.EnvelopeMinDecayTime=^FPSMix.Worldization.EnvelopeMinDecayTime
    UE.Source.OnPlay=^FireSingleShot
    LowAmmo.Volume=^LowAmmo.Volume
    MainFire.Volume=^MainFire.Volume
    Mech.Volume=^Mech.Volume
    OnStopFire.Volume=^OnStopFire.Volume
    Shells.Volume=^Shells.Volume
    Suppressor.Mix.Suppressor=^Suppressor.BaseWeaponMix
    Suppressor.Envelope.DecayTime=^Suppressor.EnvelopeDecayTime
    Suppressor.MasterVolume=^Suppressor.MasterVolume
    Suppressor.Mix.BaseWeapon=^Suppressor.SuppressorVolume
    Sweetener.Volume=^Sweetener.Volume
    Weapon.FireRate=^Weapon.FireRate
    Weapon.IsSuppressorEnabled=^Weapon.IsSuppressorEnabled
    MainFire.C=@AudioMixerAudioMixerMono2_0.Out
    Mech.L=@AudioMixerAudioMixerMono2_2.Out
    MainFire.L=@BiquadFilter_0.Out
    MainFire.R=@BiquadFilter_1.Out
    LowAmmo.C=@n18.LowAmmo.C
    Shells.L=@n19.Mech.L
    Shells.R=@n19.Mech.R
    Suppressor.L=@n20.OutL
    Suppressor.R=@n20.OutR
    Sweetener.L=@n21.Sweetener.L
    Sweetener.R=@n21.Sweetener.R
    MainFire.LFE=UE.Multiply:Audio by Float(AdditionalOperands=^FPSMix.LFEVolume, PrimaryOperand=@n2.MainFire.LFE).Out
    MainFire.Rs=UE.Multiply:Audio by Float(AdditionalOperands=^FPSMix.RightSurroundVolume, PrimaryOperand=@n2.MainFire.Rs).Out
    MainFire.Ls=UE.Multiply:Audio by Float(AdditionalOperands=^FPSMix.LeftSurroundVolume, PrimaryOperand=@n2.MainFire.Ls).Out
    OnGunshot=@n4.Value
    Mech.R=UE.WaveShaper:Audio(Amount=@n13.FloatOut, In=@n14.Mech.R) {also {defaults {Type=4}}}.Out
   )
   Multiply_14 = UE.Multiply:Audio by Float(AdditionalOperands=^FPSMix.7 Weapon Output Volume, PrimaryOperand=@n22.Out.L)
   n23 = UE.Multiply:Audio by Float(AdditionalOperands=^FPSMix.7 Weapon Output Volume, PrimaryOperand=@n22.Out.R)
   AudioBusWriter1_0 = "UE.Audio Bus Writer (1)"(In 0=UE.Multiply:Audio by Float(PrimaryOperand=UE.Add:Audio(PrimaryOperand=@Multiply_14.Out, AdditionalOperands=@n23.Out).Out) {also {defaults {AdditionalOperands=0.15}}}.Out) {also {defaults {Audio Bus=/Game/Sounds/Modulation/EnvelopeFollowers/FortWeapon/AB_WeaponFire}}}
   TriggerOnValueChange_0 = UE.Trigger On Value Change:Bool(Value=^Weapon.IsAimingDownSights)
   TriggerAnyTriggerAny2_2 = "TriggerAny.Trigger Any (2)"(In 1=@TriggerOnValueChange_0.Trigger, In 0=@n9.Value)
   n24 = &MSP_SetFloatWithBool(bBool=^Weapon.IsAimingDownSights, TriggerCompare=@TriggerAnyTriggerAny2_2.Out) {also {defaults {FloatIfFalse=0.1}}}
   Multiply_8 = UE.Multiply:Audio by Float(AdditionalOperands=^FPSMix.7 Weapon Output Volume, PrimaryOperand=@n22.Out.C)
   AudioMixerAudioMixerMono2_1 = "AudioMixer.Audio Mixer (Mono, 2)"(
    In 1=UE.Multiply:Audio by Float(AdditionalOperands=UE.InterpTo:Audio(Target=@n24.FloatOut).Value, PrimaryOperand=@Multiply_8.Out).Out
    In 0=@n23.Out
   )
   Multiply_1 = UE.Multiply:Audio by Float(
    AdditionalOperands=^FPSMix.7 Weapon Output Volume
    PrimaryOperand=UE.Multiply:Audio by Float(
     AdditionalOperands=&MSP_SetFloatWithBool(FloatIfTrue=^FPSMix.ADS.LFEVolume, bBool=^Weapon.IsAimingDownSights, TriggerCompare=@n8.Value).FloatOut
     PrimaryOperand=@n22.Out.LFE
    ).Out
   )
   Multiply_4 = UE.Multiply:Audio by Float(AdditionalOperands=^FPSMix.7 Weapon Output Volume, PrimaryOperand=@n22.Out.Rs)
   Multiply_7 = UE.Multiply:Audio by Float(AdditionalOperands=^FPSMix.7 Weapon Output Volume, PrimaryOperand=@n22.Out.Ls)
   n25 = InitVariable.Trigger()
   n26 = InitVariable.Trigger()
   TriggerAccumulatorTriggerAccumulate2_0 = "TriggerAccumulator.Trigger Accumulate (2)"(
    In 1=VariableAccessor.Trigger(Variable=VariableMutator.Trigger(Variable=@n25.Variable, Value=@n18.OnFinished).Variable).Value
    In 0=VariableAccessor.Trigger(Variable=VariableMutator.Trigger(Variable=@n26.Variable, Value=@n2.OnFinished).Variable).Value
   ) {also {defaults {Auto Reset="<MetaSoundBuilderResult.SUCCEEDED: 0>"}}}
   n27 = &MSP_TriggerResolve(A=@TriggerAccumulatorTriggerAccumulate2_0.Out, B=VariableAccessor.Trigger(Variable=@n16.Variable).Value)
   TriggerDelay_0 = UE.Trigger Delay(Delay Time=^RetriggerDelayTime, In=@TriggerAccumulatorTriggerAccumulate2_0.Out)
   TriggerControl_0 = UE.Trigger Control(Open=@n27.A, Close=@n27.B, Trigger In=@TriggerDelay_0.Out)
   ValueValue_0 = Value.Value:Bool(Set=@TriggerControl_0.Trigger Out)
  ]
  links [
   {dst="@UE.OutputFormat.5dot1.Audio:1", src=AudioMixerAudioMixerMono2_1.Out}
   {dst="@OnGunshotOut", src=MSP_Weapon_Fire_Control_1p_0.OutGunshot}
   {dst="@UE.OutputFormat.5dot1.Audio:3", src=Multiply_1.Out}
   {dst="@UE.OutputFormat.5dot1.Audio:0", src=Multiply_14.Out}
   {dst="@UE.OutputFormat.5dot1.Audio:5", src=Multiply_4.Out}
   {dst="@UE.OutputFormat.5dot1.Audio:4", src=Multiply_7.Out}
   {dst="@UE.OutputFormat.5dot1.Audio:2", src=Multiply_8.Out}
   {dst="@UE.Source.OneShot.OnFinished", src=TriggerControl_0.Trigger Out}
   {dst="@bOnAlmostFinished", src=ValueValue_0.Output Value}
  ]
 }
 Mobile {
  input_defaults {bIsFiring=null}
  nodes [
   MSP_Weapon_Fire_Control_1p_0 = &MSP_Weapon_Fire_Control_1p(
    FireBurst=^FireBurst
    FireSingleShot=^FireSingleShot
    RoundsPerBurst=^RoundsPerBurst
    StopBurst=^StopBurst
    OnPlay=^UE.Source.OnPlay
    Weapon.AmmoRemaining=^Weapon.AmmoRemaining
    Weapon.FireRate=^Weapon.FireRate
    bIsFiring=^bIsFiring
   )
   n1 = VariableAccessor.Trigger(Variable=VariableMutator.Trigger(Variable=InitVariable.Trigger().Variable, Value=@MSP_Weapon_Fire_Control_1p_0.OutGunshot).Variable)
   n2 = &MSP_Weapon_Fire_MainFire(
    MainFire.ADS.VolumeMultiply=^MainFire.ADS.VolumeMultiply
    MainFire.Random.Pitch.Min=^MainFire.Random.Pitch.Max
    MainFire.Random.Pitch.Max=^MainFire.Random.Pitch.Min
    MainFire.Random.Volume.Max=^MainFire.Random.Volume.Max
    MainFire.Random.Volume.Min=^MainFire.Random.Volume.Min
    MainFire.Sounds=^MainFire.Sounds
    Weapon.IsAimingDownSights=^Weapon.IsAimingDownSights
    OnGunshot=@n1.Value
   ) {also {defaults {bUseSingleVoice="<MetaSoundBuilderResult.SUCCEEDED: 0>"}}}
   TriggerOnValueChange_1 = UE.Trigger On Value Change:Bool()
   n3 = VariableAccessor.Trigger(Variable=@n1.Variable)
   n4 = VariableAccessor.Trigger(Variable=@n3.Variable)
   n5 = VariableAccessor.Trigger(Variable=@n4.Variable)
   n6 = VariableAccessor.Trigger(Variable=@n5.Variable)
   n7 = VariableAccessor.Trigger(Variable=@n6.Variable)
   n8 = VariableAccessor.Trigger(Variable=@n7.Variable)
   n9 = VariableAccessor.Trigger(Variable=@n8.Variable)
   TriggerAnyTriggerAny2_1 = "TriggerAny.Trigger Any (2)"(In 1=@TriggerOnValueChange_1.Trigger, In 0=@n9.Value)
   AudioMixerAudioMixerMono2_0 = "AudioMixer.Audio Mixer (Mono, 2)"(
    In 0=UE.Multiply:Audio by Float(AdditionalOperands=^FPSMix.CenterVolume, PrimaryOperand=@n2.MainFire.C).Out
    In 1=UE.WaveShaper:Audio(
     Amount=&MSP_SetFloatWithBool(
      FloatIfTrue=^FPSMIX.ADS.Filter.Wave Shaper Main Fire Sub ADS Fire Amount
      bBool=^Weapon.IsAimingDownSights
      TriggerCompare=@TriggerAnyTriggerAny2_1.Out
     ).FloatOut
     In=@n2.MainFire.LFE
    ) {also {defaults {Type=4}}}.Out
   ) {also {defaults {Gain 1=0.6}}}
   TriggerOnValueChange_2 = UE.Trigger On Value Change:Bool(Value=^Weapon.IsAimingDownSights)
   n10 = VariableAccessor.Trigger(Variable=@n9.Variable)
   TriggerAnyTriggerAny2_3 = "TriggerAny.Trigger Any (2)"(In 1=@TriggerOnValueChange_2.Trigger, In 0=@n10.Value)
   TriggerOnValueChange_3 = UE.Trigger On Value Change:Bool(Value=^Weapon.IsAimingDownSights)
   TriggerAnyTriggerAny2_0 = "TriggerAny.Trigger Any (2)"(In 1=@TriggerOnValueChange_3.Trigger, In 0=@n8.Value)
   n11 = &MSP_SetFloatWithBool(
    FloatIfTrue=^FPSMIX.ADS.Filter.Wave Shaper Mech ADS Fire Amount
    bBool=^Weapon.IsAimingDownSights
    TriggerCompare=@TriggerAnyTriggerAny2_0.Out
   )
   n12 = &MSP_Weapon_Fire_Mech_FPS(
    Mech.ADS.VolumeMultiply=^Mech.ADSVolumeMultiply
    Mech.Random.Pitch.Max=^Mech.Random.Pitch.Max
    Mech.Random.Pitch.Min=^Mech.Random.Pitch.Min
    Mech.Random.Volume.Max=^Mech.Random.Volume.Max
    Mech.Random.Volume.Min=^Mech.Random.Volume.Min
    Mech.Sounds=^Mech.Sounds
    Weapon.IsAimingDownSights=^Weapon.IsAimingDownSights
    OnGunshot=@n4.Value
   ) {also {defaults {bUseSingleVoice="<MetaSoundBuilderResult.SUCCEEDED: 0>"}}}
   AudioMixerAudioMixerMono2_2 = "AudioMixer.Audio Mixer (Mono, 2)"(
    Gain 0=UE.InterpTo:Audio(Target=&MSP_SetFloatWithBool(bBool=^Weapon.IsAimingDownSights, TriggerCompare=@TriggerAnyTriggerAny2_3.Out) {also {defaults {FloatIfFalse=0.75, FloatIfTrue=1}}}.FloatOut).Value
    In 0=UE.WaveShaper:Audio(Amount=@n11.FloatOut, In=@n12.Mech.L) {also {defaults {Type=4}}}.Out
   ) {also {defaults {Gain 1=0}}}
   n13 = VariableAccessor.Trigger(Variable=@n10.Variable)
   LinearGaintoDecibels_0 = UE.Linear Gain to Decibels:Float(Linear Gain=&MSP_SetFloatWithBool(
     FloatIfTrue=^FPSMIX.ADS.Filter.GainADS
     FloatIfFalse=^FPSMIX.ADS.Filter.GainHipFire
     bBool=^Weapon.IsAimingDownSights
     TriggerCompare=@n13.Value
    ).FloatOut)
   BiquadFilter_0 = UE.Biquad Filter:Audio(
    Bandwidth=^FPSMIX.ADS.Filter.Bandwidth
    Cutoff Frequency=^FPSMIX.ADS.Filter.CutoffFreq
    Gain=@LinearGaintoDecibels_0.Decibels
    In=UE.Multiply:Audio by Float(AdditionalOperands=^FPSMix.LeftVolume, PrimaryOperand=@n2.MainFire.L).Out
   ) {also {defaults {Type=4}}}
   BiquadFilter_1 = UE.Biquad Filter:Audio(
    Bandwidth=^FPSMIX.ADS.Filter.Bandwidth
    Cutoff Frequency=^FPSMIX.ADS.Filter.CutoffFreq
    Gain=@LinearGaintoDecibels_0.Decibels
    In=UE.Multiply:Audio by Float(AdditionalOperands=^FPSMix.RightVolume, PrimaryOperand=@n2.MainFire.R).Out
   ) {also {defaults {Type=4}}}
   n14 = InitVariable.Int32()
   n15 = &MSP_Weapon_Fire_LowAmmo(
    LowAmmo.ADS.VolumeMultiply=^LowAmmo.ADS.VolumeMultiply
    LowAmmo.Random.Pitch.Max=^LowAmmo.Random.Pitch.Max
    LowAmmo.Random.Pitch.Min=^LowAmmo.Random.Pitch.Min
    LowAmmo.Random.Volume.Max=^LowAmmo.Random.Volume.Max
    LowAmmo.Random.Volume.Min=^LowAmmo.Random.Volume.Min
    LowAmmo.Sounds=^LowAmmo.Sounds
    LowAmmo.Threshold=^LowAmmo.Threshold
    Weapon.IsAimingDownSights=^Weapon.IsAimingDownSights
    CurrentAmmo=VariableAccessor.Int32(Variable=VariableMutator.Int32(Variable=@n14.Variable, Value=@MSP_Weapon_Fire_Control_1p_0.CurrentAmmo).Variable).Value
    OnGunshot=@n5.Value
   )
   n16 = VariableAccessor.Trigger(Variable=@n13.Variable)
   n17 = &MSP_Weapon_Fire_Suppressor(
    Random.Pitch.Max=^Suppressor.Random.Pitch.Max
    Random.Pitch.Min=^Suppressor.Random.Pitch.Min
    Random.Volume.Max=^Suppressor.Random.Volume.Max
    Random.Volume.Min=^Suppressor.Random.Volume.Min
    Sounds=^Suppressor.Sounds
    Weapon.IsSuppressorEnabled=^Weapon.IsSuppressorEnabled
    OnGunshot=@n16.Value
   )
   n18 = &MSP_Weapon_Fire_FPS_Mixing_1P(
    Worldization.ActivationThreshold=^FPSMix.Worldization.ActivationThreshold
    Worldization.EnvelopeMaxDecayTime=^FPSMix.Worldization.EnvelopeMaxDecayTime
    Worldization.EnvelopeMinDecayTime=^FPSMix.Worldization.EnvelopeMinDecayTime
    UE.Source.OnPlay=^FireSingleShot
    LowAmmo.Volume=^LowAmmo.Volume
    MainFire.Volume=^MainFire.Volume
    Mech.Volume=^Mech.Volume
    OnStopFire.Volume=^OnStopFire.Volume
    Suppressor.Mix.Suppressor=^Suppressor.BaseWeaponMix
    Suppressor.Envelope.DecayTime=^Suppressor.EnvelopeDecayTime
    Suppressor.MasterVolume=^Suppressor.MasterVolume
    Suppressor.Mix.BaseWeapon=^Suppressor.SuppressorVolume
    Weapon.FireRate=^Weapon.FireRate
    Weapon.IsSuppressorEnabled=^Weapon.IsSuppressorEnabled
    MainFire.C=@AudioMixerAudioMixerMono2_0.Out
    Mech.L=@AudioMixerAudioMixerMono2_2.Out
    MainFire.L=@BiquadFilter_0.Out
    MainFire.R=@BiquadFilter_1.Out
    LowAmmo.C=@n15.LowAmmo.C
    Suppressor.L=@n17.OutL
    Suppressor.R=@n17.OutR
    MainFire.LFE=UE.Multiply:Audio by Float(AdditionalOperands=^FPSMix.LFEVolume, PrimaryOperand=@n2.MainFire.LFE).Out
    MainFire.Rs=UE.Multiply:Audio by Float(AdditionalOperands=^FPSMix.RightSurroundVolume, PrimaryOperand=@n2.MainFire.Rs).Out
    MainFire.Ls=UE.Multiply:Audio by Float(AdditionalOperands=^FPSMix.LeftSurroundVolume, PrimaryOperand=@n2.MainFire.Ls).Out
    OnGunshot=@n3.Value
    Mech.R=UE.WaveShaper:Audio(Amount=@n11.FloatOut, In=@n12.Mech.R) {also {defaults {Type=4}}}.Out
   )
   Multiply_14 = UE.Multiply:Audio by Float(AdditionalOperands=^FPSMix.7 Weapon Output Volume, PrimaryOperand=@n18.Out.L)
   n19 = UE.Multiply:Audio by Float(AdditionalOperands=^FPSMix.7 Weapon Output Volume, PrimaryOperand=@n18.Out.R)
   AudioBusWriter1_0 = "UE.Audio Bus Writer (1)"(In 0=UE.Multiply:Audio by Float(PrimaryOperand=UE.Add:Audio(PrimaryOperand=@Multiply_14.Out, AdditionalOperands=@n19.Out).Out) {also {defaults {AdditionalOperands=0.15}}}.Out) {also {defaults {Audio Bus=&AB_WeaponFire}}}
   TriggerOnValueChange_0 = UE.Trigger On Value Change:Bool(Value=^Weapon.IsAimingDownSights)
   TriggerAnyTriggerAny2_2 = "TriggerAny.Trigger Any (2)"(In 1=@TriggerOnValueChange_0.Trigger, In 0=@n7.Value)
   n20 = &MSP_SetFloatWithBool(bBool=^Weapon.IsAimingDownSights, TriggerCompare=@TriggerAnyTriggerAny2_2.Out) {also {defaults {FloatIfFalse=0.1}}}
   Multiply_8 = UE.Multiply:Audio by Float(AdditionalOperands=^FPSMix.7 Weapon Output Volume, PrimaryOperand=@n18.Out.C)
   AudioMixerAudioMixerMono2_1 = "AudioMixer.Audio Mixer (Mono, 2)"(
    In 1=UE.Multiply:Audio by Float(AdditionalOperands=UE.InterpTo:Audio(Target=@n20.FloatOut).Value, PrimaryOperand=@Multiply_8.Out).Out
    In 0=@n19.Out
   )
   Multiply_1 = UE.Multiply:Audio by Float(
    AdditionalOperands=^FPSMix.7 Weapon Output Volume
    PrimaryOperand=UE.Multiply:Audio by Float(
     AdditionalOperands=&MSP_SetFloatWithBool(FloatIfTrue=^FPSMix.ADS.LFEVolume, bBool=^Weapon.IsAimingDownSights, TriggerCompare=@n6.Value).FloatOut
     PrimaryOperand=@n18.Out.LFE
    ).Out
   )
   Multiply_4 = UE.Multiply:Audio by Float(AdditionalOperands=^FPSMix.7 Weapon Output Volume, PrimaryOperand=@n18.Out.Rs)
   Multiply_7 = UE.Multiply:Audio by Float(AdditionalOperands=^FPSMix.7 Weapon Output Volume, PrimaryOperand=@n18.Out.Ls)
   n21 = InitVariable.Trigger()
   n22 = InitVariable.Trigger()
   TriggerAccumulatorTriggerAccumulate2_0 = "TriggerAccumulator.Trigger Accumulate (2)"(
    In 0=VariableAccessor.Trigger(Variable=VariableMutator.Trigger(Variable=@n21.Variable, Value=@n2.OnFinished).Variable).Value
    In 1=VariableAccessor.Trigger(Variable=VariableMutator.Trigger(Variable=@n22.Variable, Value=@n15.OnFinished).Variable).Value
   ) {also {defaults {Auto Reset="<MetaSoundBuilderResult.SUCCEEDED: 0>"}}}
   n23 = &MSP_TriggerResolve(A=@TriggerAccumulatorTriggerAccumulate2_0.Out, B=VariableAccessor.Trigger(Variable=@n16.Variable).Value)
   TriggerDelay_0 = UE.Trigger Delay(Delay Time=^RetriggerDelayTime, In=@TriggerAccumulatorTriggerAccumulate2_0.Out)
   TriggerControl_0 = UE.Trigger Control(Open=@n23.A, Close=@n23.B, Trigger In=@TriggerDelay_0.Out)
   ValueValue_0 = Value.Value:Bool(Set=@TriggerControl_0.Trigger Out)
  ]
  links [
   {dst="@UE.OutputFormat.5dot1.Audio:1", src=AudioMixerAudioMixerMono2_1.Out}
   {dst="@OnGunshotOut", src=MSP_Weapon_Fire_Control_1p_0.OutGunshot}
   {dst="@UE.OutputFormat.5dot1.Audio:3", src=Multiply_1.Out}
   {dst="@UE.OutputFormat.5dot1.Audio:0", src=Multiply_14.Out}
   {dst="@UE.OutputFormat.5dot1.Audio:5", src=Multiply_4.Out}
   {dst="@UE.OutputFormat.5dot1.Audio:4", src=Multiply_7.Out}
   {dst="@UE.OutputFormat.5dot1.Audio:2", src=Multiply_8.Out}
   {dst="@UE.Source.OneShot.OnFinished", src=TriggerControl_0.Trigger Out}
   {dst="@bOnAlmostFinished", src=ValueValue_0.Output Value}
  ]
 }
]
properties {
 assetImportData=&self:AssetImportData
 numChannels=6
 outputFormat=FiveDotOne
 qualitySetting=Fort_Weapons
 sampleRate=48000
 soundClassObject=&Fort_WeaponRanged_Fire_1P_PSM
 concurrencyOverrides {maxCount=2, resolutionRule=StopOldest}
 concurrencySet [/FortMetaSoundWeapon/Items/Shared/Settings/FPS/CS_Weapon_FP_Gunfire_Local]
 modulationSettings {
  volumeRouting=Union
  volumeModulationDestination {bEnableModulation=true, modulators [/FortMetaSoundWeapon/Items/Modulation/PP_Weapon_Fire_1p_Vol]}
 }
}
