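# Legacy MetaSound patch that generates a seeded randomization profile for weapon fire sounds, covering volume, pitch and two EQ bands.
# InitRandom fires chained RandomFloat nodes, all sharing SeedProfile, which also passes through as an output.
# Volume and pitch ranges split at their midpoints, so each randomized minimum stays below its maximum.
# EQ frequency and gain use input ranges, while bandwidths and UniquePitch use fixed ranges.
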MetaSoundPatch /FortMetaSoundWeapon/Items/Metasounds/Legacy/Subgraphs/MSP_Weapon_Fire_RandomSeed
inputs [
 InitRandom {type=Trigger}
 Random.EQ.A.MaxFreq {type=Float}
 Random.EQ.A.MaxGain {type=Float}
 Random.EQ.A.MinFreq {type=Float}
 Random.EQ.A.MinGain {type=Float}
 Random.EQ.B.MaxFreq {type=Float}
 Random.EQ.B.MaxGain {type=Float}
 Random.EQ.B.MinFreq {type=Float}
 Random.EQ.B.MinGain {type=Float}
 Random.MaxPitch {type=Float}
 Random.MaxVolume {type=Float}
 Random.MinPitch {type=Float}
 Random.MinVolume {type=Float}
 SeedProfile {type=Int32}
]
outputs [
 MaxPitchProfile {type=Float}
 MaxVolumeProfile {type=Float}
 MinPitchProfile {type=Float}
 MinVolumeProfile {type=Float}
 RandomBandwidthA {type=Float}
 RandomBandwidthB {type=Float}
 RandomFrequencyA {type=Float}
 RandomFrequencyB {type=Float}
 RandomGainA {type=Float}
 RandomGainB {type=Float}
 SeedProfile {type=Int32}
 UniquePitch {type=Float}
]
pages [Default {
  input_defaults {
   InitRandom=null
   Random.EQ.A.MaxFreq=20000
   Random.EQ.A.MaxGain=0
   Random.EQ.A.MinFreq=10000
   Random.EQ.A.MinGain=-1
   Random.EQ.B.MaxFreq=8000
   Random.EQ.B.MaxGain=0
   Random.EQ.B.MinFreq=1000
   Random.EQ.B.MinGain=-1
   Random.MaxPitch=1.15
   Random.MaxVolume=1.15
   Random.MinPitch=0.85
   Random.MinVolume=0.85
   SeedProfile=-1
  }
  nodes [
   VariableAccessorInt32_8 = VariableAccessor.Int32(Variable=VariableMutator.Int32(Value=^SeedProfile, Variable=InitVariable.Int32().Variable).Variable)
   n1 = VariableAccessor.Int32(Variable=@VariableAccessorInt32_8.Variable)
   n2 = VariableAccessor.Int32(Variable=@n1.Variable)
   n3 = VariableAccessor.Int32(Variable=@n2.Variable)
   n4 = VariableAccessor.Int32(Variable=@n3.Variable)
   n5 = VariableAccessor.Int32(Variable=@n4.Variable)
   n6 = VariableAccessor.Trigger(Variable=VariableMutator.Trigger(Value=^InitRandom, Variable=InitVariable.Trigger().Variable).Variable)
   n7 = VariableAccessor.Trigger(Variable=@n6.Variable)
   n8 = VariableAccessor.Trigger(Variable=@n7.Variable)
   RandomFloat_7 = UE.RandomFloat(Max=^Random.EQ.A.MaxFreq, Min=^Random.EQ.A.MinFreq, Seed=@n5.Value, Next=@n8.Value)
   n9 = VariableAccessor.Int32(Variable=@n5.Variable)
   RandomFloat_10 = UE.RandomFloat(Next=@RandomFloat_7.On Next, Seed=@n9.Value) {also {defaults {Max=1.2, Min=0.1}}}
   n10 = VariableAccessor.Int32(Variable=@n9.Variable)
   RandomFloat_2 = UE.RandomFloat(Max=^Random.EQ.A.MaxGain, Min=^Random.EQ.A.MinGain, Next=@RandomFloat_10.On Next, Seed=@n10.Value)
   n11 = VariableAccessor.Int32(Variable=@n10.Variable)
   n12 = VariableAccessor.Trigger(Variable=@n8.Variable)
   RandomFloat_1 = UE.RandomFloat(Max=^Random.EQ.B.MaxFreq, Min=^Random.EQ.B.MinFreq, Seed=@n11.Value, Next=@n12.Value)
   n13 = VariableAccessor.Int32(Variable=@n11.Variable)
   RandomFloat_0 = UE.RandomFloat(Next=@RandomFloat_1.On Next, Seed=@n13.Value) {also {defaults {Max=1.2, Min=0.1}}}
   n14 = VariableAccessor.Int32(Variable=@n13.Variable)
   RandomFloat_4 = UE.RandomFloat(Max=^Random.EQ.B.MaxGain, Min=^Random.EQ.B.MinGain, Next=@RandomFloat_0.On Next, Seed=@n14.Value)
   n15 = UE.Multiply:Float(PrimaryOperand=UE.Add:Float(AdditionalOperands=^Random.MaxVolume, PrimaryOperand=^Random.MinVolume).Out) {also {defaults {AdditionalOperands=0.5}}}
   RandomFloat_5 = UE.RandomFloat(Min=^Random.MinVolume, Max=@n15.Out, Seed=@n1.Value, Next=@n6.Value)
   RandomFloat_6 = UE.RandomFloat(
    Min=UE.Add:Float(PrimaryOperand=@n15.Out) {also {defaults {AdditionalOperands=0.01}}}.Out
    Next=@RandomFloat_5.On Next
    Max=UE.Reroute(Value=UE.Reroute(Value=^Random.MaxVolume).Value).Value
    Seed=@n2.Value
   )
   n16 = UE.Multiply:Float(PrimaryOperand=UE.Add:Float(AdditionalOperands=^Random.MaxPitch, PrimaryOperand=^Random.MinPitch).Out) {also {defaults {AdditionalOperands=0.5}}}
   RandomFloat_3 = UE.RandomFloat(Min=^Random.MinPitch, Max=@n16.Out, Seed=@n3.Value, Next=@n7.Value)
   RandomFloat_8 = UE.RandomFloat(
    Min=UE.Add:Float(PrimaryOperand=@n16.Out) {also {defaults {AdditionalOperands=0.01}}}.Out
    Next=@RandomFloat_3.On Next
    Max=UE.Reroute(Value=UE.Reroute(Value=^Random.MaxPitch).Value).Value
    Seed=@n4.Value
   )
   RandomFloat_9 = UE.RandomFloat(Seed=VariableAccessor.Int32(Variable=@n14.Variable).Value, Next=VariableAccessor.Trigger(Variable=@n12.Variable).Value) {also {defaults {Max=1.1, Min=0.91}}}
  ]
  links [
   {dst="@RandomBandwidthB", src=RandomFloat_0.Value}
   {dst="@RandomFrequencyB", src=RandomFloat_1.Value}
   {dst="@RandomBandwidthA", src=RandomFloat_10.Value}
   {dst="@RandomGainA", src=RandomFloat_2.Value}
   {dst="@MinPitchProfile", src=RandomFloat_3.Value}
   {dst="@RandomGainB", src=RandomFloat_4.Value}
   {dst="@MinVolumeProfile", src=RandomFloat_5.Value}
   {dst="@MaxVolumeProfile", src=RandomFloat_6.Value}
   {dst="@RandomFrequencyA", src=RandomFloat_7.Value}
   {dst="@MaxPitchProfile", src=RandomFloat_8.Value}
   {dst="@UniquePitch", src=RandomFloat_9.Value}
   {dst="@SeedProfile", src=VariableAccessorInt32_8.Value}
  ]
 }]
