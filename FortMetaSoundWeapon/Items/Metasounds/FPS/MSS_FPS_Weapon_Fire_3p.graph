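# Stereo one-shot MetaSound source for third-person FPS weapon gunfire, with Default and Mobile pages.
# A control subgraph turns single, burst and automatic fire inputs into gunshots driving close, medium, distant and suppressor layers, randomized per character ID.
# A mixing subgraph combines layers using distance curves, affiliation and elevation, feeding the output and an envelope-follower bus.
# OnFinished fires after all layers end plus a retrigger delay, gated against new shots.
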MetaSoundSource /FortMetaSoundWeapon/Items/Metasounds/FPS/MSS_FPS_Weapon_Fire_3p
inputs [
 BurstRate {type=Time}
 bUseSingleVoice {type=Bool}
 Character.Affiliation {type=Enum:FortAudioAffiliation}
 Character.Health {type=Float}
 Character.UniqueID {type=Int32}
 DistanceEQ.bEnable {type=Bool}
 DistanceEQ.FilterMaxGain {type=Float}
 DistanceEQ.FilterMinGain {type=Float}
 DistanceEQ.HighShelfCutoffFrequency {type=Float}
 DistanceEQ.LowShelfCutoffFrequency {type=Float}
 Elevation.AboveGain {type=Float}
 Elevation.BelowGain {type=Float}
 Elevation.CutoffFrequency {type=Float}
 Elevation.LowCutGain {type=Float}
 FireBurst {type=Trigger}
 FireSingleShot {type=Trigger}
 MainFire.DistanceCurves {type=WaveTableBankAsset}
 MainFire.FalloffDistance {type=Float}
 MainFire.SoundsClose {type=WaveAsset:Array}
 MainFire.SoundsDistant {type=WaveAsset:Array}
 MainFire.SoundsMedium {type=WaveAsset:Array}
 Mixing.CloseGain {type=Float}
 Mixing.DistantGain {type=Float}
 Mixing.Friendly.FriendlyFireWeaponVolume {type=Float}
 Mixing.MediumGain {type=Float}
 Mixing.Suppressor.SuppressorOutputVolume {type=Float}
 Mixing.WeaponOuputVolume {type=Float}
 Random.EQ.A.MaxFreq {type=Float}
 Random.EQ.A.MaxGain {type=Float}
 Random.EQ.A.MinFreq {type=Float}
 Random.EQ.A.MinGain {type=Float}
 Random.EQ.B.MaxFreq {type=Float}
 Random.EQ.B.MaxGain {type=Float}
 Random.EQ.B.MinFreq {type=Float}
 Random.EQ.B.MinGain {type=Float}
 Random.MaxPitch {type=Float}
 Random.MaxVolume {type=Float}
 Random.MinPitch {type=Float}
 Random.MinVolume {type=Float}
 RetriggerDelayTime {type=Time}
 RoundsPerBurst {type=Int32}
 StartFiring {type=Trigger}
 StopBurst {type=Trigger}
 StopFiring {type=Trigger}
 Suppressor.BaseWeaponVolumeMultiplier {type=Float}
 Suppressor.EnvelopeAttackTime {type=Time}
 Suppressor.EnvelopeDecayTime {type=Time}
 Suppressor.Sounds.Close {type=WaveAsset:Array}
 Suppressor.Sounds.Medium {type=WaveAsset:Array}
 Suppressor.Volume.Close {type=Float}
 Suppressor.Volume.CloseBase {type=Float}
 Suppressor.Volume.Medium {type=Float}
 Suppressor.Volume.MediumBase {type=Float}
 TargetEQ.NonTargetGain {type=Float}
 UE.Attenuation.Distance {type=Float}
 UE.Source.OnPlay {type=Trigger}
 UE.Source.Orientation.Azimuth {type=Float}
 UE.Source.Orientation.Elevation {type=Float}
 UE.Spatialization.Azimuth {type=Float}
 UE.Spatialization.Elevation {type=Float}
 Weapon.AmmoRemaining {type=Int32}
 Weapon.ClipSize {type=Int32}
 Weapon.FireRate {type=Float}
 Weapon.FireStartTime {type=Float}
 Weapon.IsAimingDownSights {type=Bool}
 Weapon.IsSuppressorEnabled {type=Bool}
]
interfaces [
 Fort.Character:1.0
 Fort.Weapon:1.0
 UE.Attenuation:1.0
 UE.OutputFormat.Stereo:1.0
 UE.Source.OneShot:1.0
 UE.Source.Orientation:1.0
 UE.Source:1.1
 UE.Spatialization:1.0
]
outputs [
 bOnAlmostFinished {type=Bool}
 CloseCurveEvaluation {type=Float}
 MediumCurveEvaluation {type=Float}
 OnGunshotOut {type=Trigger}
 "UE.OutputFormat.Stereo.Audio:0" {type=Audio}
 "UE.OutputFormat.Stereo.Audio:1" {type=Audio}
 UE.Source.OneShot.OnFinished {type=Trigger}
]
pages [
 Default {
  is_build_graph=true
  input_defaults {
   BurstRate=0.4
   Character.Affiliation=0
   Character.Health=1
   Character.UniqueID=0
   DistanceEQ.FilterMaxGain=1
   DistanceEQ.FilterMinGain=-3
   DistanceEQ.HighShelfCutoffFrequency=11000
   DistanceEQ.LowShelfCutoffFrequency=180
   DistanceEQ.bEnable="<MetaSoundBuilderResult.SUCCEEDED: 0>"
   Elevation.AboveGain=0.4
   Elevation.BelowGain=-1
   Elevation.CutoffFrequency=1500
   Elevation.LowCutGain=-6
   FireBurst=null
   FireSingleShot=null
   MainFire.DistanceCurves=/FortMetaSoundWeapon/Items/Shared/Settings/FPS/WTB_Weapon_FP_DistanceCurves
   MainFire.FalloffDistance=25000
   Mixing.CloseGain=1
   Mixing.DistantGain=0.6
   Mixing.Friendly.FriendlyFireWeaponVolume=0.6
   Mixing.MediumGain=0.7
   Mixing.Suppressor.SuppressorOutputVolume=1
   Mixing.WeaponOuputVolume=0.85
   Random.EQ.A.MaxFreq=20000
   Random.EQ.A.MaxGain=0
   Random.EQ.A.MinFreq=699.9999
   Random.EQ.A.MinGain=-3.5
   Random.EQ.B.MaxFreq=8000
   Random.EQ.B.MaxGain=0
   Random.EQ.B.MinFreq=700
   Random.EQ.B.MinGain=-2
   Random.MaxPitch=1.15
   Random.MaxVolume=1.15
   Random.MinPitch=0.85
   Random.MinVolume=0.85
   RetriggerDelayTime=2
   RoundsPerBurst=2
   StartFiring=null
   StopBurst=null
   StopFiring=null
   Suppressor.BaseWeaponVolumeMultiplier=0.75
   Suppressor.EnvelopeAttackTime=0
   Suppressor.EnvelopeDecayTime=0.195
   Suppressor.Volume.Close=0.5
   Suppressor.Volume.CloseBase=1
   Suppressor.Volume.Medium=0.3
   Suppressor.Volume.MediumBase=0.3
   TargetEQ.NonTargetGain=-2
   UE.Attenuation.Distance=27500
   UE.Source.OnPlay=null
   UE.Source.Orientation.Azimuth=0
   UE.Source.Orientation.Elevation=0
   UE.Spatialization.Azimuth=0
   UE.Spatialization.Elevation=0
   Weapon.AmmoRemaining=30
   Weapon.ClipSize=30
   Weapon.FireRate=0.5
   Weapon.FireStartTime=0
   Weapon.IsAimingDownSights=null
   Weapon.IsSuppressorEnabled=null
   bUseSingleVoice="<MetaSoundBuilderResult.SUCCEEDED: 0>"
   MainFire.SoundsClose []
   MainFire.SoundsDistant []
   MainFire.SoundsMedium []
   Suppressor.Sounds.Close []
   Suppressor.Sounds.Medium []
  }
  nodes [
   n1 = /FortMetaSoundWeapon/Items/Metasounds/Legacy/Subgraphs/MSP_Weapon_Fire_RandomSeed(
    SeedProfile=^Character.UniqueID
    Random.EQ.A.MaxFreq=^Random.EQ.A.MaxFreq
    Random.EQ.A.MaxGain=^Random.EQ.A.MaxGain
    Random.EQ.A.MinFreq=^Random.EQ.A.MinFreq
    Random.EQ.A.MinGain=^Random.EQ.A.MinGain
    Random.EQ.B.MaxFreq=^Random.EQ.B.MaxFreq
    Random.EQ.B.MaxGain=^Random.EQ.B.MaxGain
    Random.EQ.B.MinFreq=^Random.EQ.B.MinFreq
    Random.EQ.B.MinGain=^Random.EQ.B.MinGain
    Random.MaxPitch=^Random.MaxPitch
    Random.MaxVolume=^Random.MaxVolume
    Random.MinPitch=^Random.MinPitch
    Random.MinVolume=^Random.MinVolume
    InitRandom=^UE.Source.OnPlay
   )
   n2 = VariableAccessor.Float(Variable=VariableMutator.Float(Variable=InitVariable.Float().Variable, Value=@n1.UniquePitch).Variable)
   n3 = VariableAccessor.Float(Variable=@n2.Variable)
   n4 = VariableAccessor.Float(Variable=@n3.Variable)
   n5 = VariableAccessor.Float(Variable=VariableMutator.Float(Variable=InitVariable.Float().Variable, Value=@n1.MinVolumeProfile).Variable)
   n6 = VariableAccessor.Float(Variable=@n5.Variable)
   n7 = VariableAccessor.Float(Variable=VariableMutator.Float(Variable=InitVariable.Float().Variable, Value=@n1.MaxVolumeProfile).Variable)
   n8 = VariableAccessor.Float(Variable=@n7.Variable)
   MSP_Weapon_Fire_Control_3p_0 = /FortMetaSoundWeapon/Items/Metasounds/Subgraphs/MSP_Weapon_Fire_Control_3p(
    FireBurst=^FireBurst
    FireSingleShot=^FireSingleShot
    RoundsPerBurst=^RoundsPerBurst
    StartFiring=^StartFiring
    StopBurst=^StopBurst
    StopFiring=^StopFiring
    Weapon.AmmoRemaining=^Weapon.AmmoRemaining
    Weapon.FireRate=^Weapon.FireRate
    BurstRate=UE.ConversionTimeToFloat(In=^BurstRate).Out
   )
   n9 = VariableAccessor.Trigger(Variable=VariableMutator.Trigger(Variable=InitVariable.Trigger().Variable, Value=@MSP_Weapon_Fire_Control_3p_0.OutGunshot).Variable)
   n10 = VariableAccessor.Trigger(Variable=@n9.Variable)
   n11 = VariableAccessor.Trigger(Variable=@n10.Variable)
   n12 = VariableAccessor.Trigger(Variable=@n11.Variable)
   n13 = /FortMetaSoundWeapon/Items/Metasounds/Subgraphs/MSP_Weapon_Fire_3p_Mono(
    Sounds=^MainFire.SoundsDistant
    bUseSingleVoice=^bUseSingleVoice
    Random.Pitch=@n4.Value
    Random.Volume.Min=@n6.Value
    Random.Volume.Max=@n8.Value
    OnGunshot=@n12.Value
   )
   n14 = &MSP_Weapon_Fire_3p_Mono(
    Sounds=^MainFire.SoundsMedium
    bUseSingleVoice=^bUseSingleVoice
    Random.Volume.Max=@n7.Value
    Random.Pitch=@n3.Value
    Random.Volume.Min=@n5.Value
    OnGunshot=@n11.Value
   )
   n15 = VariableAccessor.Float(Variable=@n8.Variable)
   n16 = VariableAccessor.Float(Variable=@n6.Variable)
   n17 = /FortMetaSoundWeapon/Items/Metasounds/Subgraphs/MSP_Weapon_Fire_3p_Stereo(
    Sounds=^MainFire.SoundsClose
    bUseSingleVoice=^bUseSingleVoice
    Random.Volume.Max=VariableAccessor.Float(Variable=@n15.Variable).Value
    Random.Volume.Min=VariableAccessor.Float(Variable=@n16.Variable).Value
    Random.Pitch=@n2.Value
    OnGunshot=@n10.Value
   )
   n18 = VariableAccessor.Trigger(Variable=@n12.Variable)
   n19 = /FortMetaSoundWeapon/Items/Metasounds/Subgraphs/MSP_Weapon_Fire_Suppressor_3p(
    Sounds.Close=^Suppressor.Sounds.Close
    Sounds.Medium=^Suppressor.Sounds.Medium
    Volume.Close=^Suppressor.Volume.Close
    Volume.Medium=^Suppressor.Volume.Medium
    bIsSuppressorEnabled=^Weapon.IsSuppressorEnabled
    bUseSingleVoice=^bUseSingleVoice
    Random.Volume.Min=@n16.Value
    Random.Pitch=VariableAccessor.Float(Variable=@n4.Variable).Value
    Random.Volume.Max=@n15.Value
    OnGunshot=@n18.Value
   )
   n20 = InitVariable.Float()
   n21 = InitVariable.Float()
   n22 = InitVariable.Float()
   n23 = /FortMetaSoundWeapon/Items/Metasounds/Subgraphs/MSP_Weapon_Fire_Mixing_3p(
    Character.Affiliation=^Character.Affiliation
    bEnableDistanceEQ=^DistanceEQ.bEnable
    FireSingleShot=^FireSingleShot
    DistanceCurves=^MainFire.DistanceCurves
    MaxFalloffDistance=^MainFire.FalloffDistance
    Volume.Close=^Mixing.CloseGain
    Volume.Distant=^Mixing.DistantGain
    Mixing.Friendly=^Mixing.Friendly.FriendlyFireWeaponVolume
    Volume.Medium=^Mixing.MediumGain
    Suppressor.Mixing.SuppressorVolume=^Mixing.Suppressor.SuppressorOutputVolume
    Suppressor.Mixing.BaseWeaponVolume=^Suppressor.BaseWeaponVolumeMultiplier
    Suppressor.EnvelopeDecay=^Suppressor.EnvelopeDecayTime
    Distance=^UE.Attenuation.Distance
    Spatialization.Elevation=^UE.Spatialization.Elevation
    Weapon.FireRate=^Weapon.FireRate
    Weapon.IsSuppressorEnabled=^Weapon.IsSuppressorEnabled
    Distant.L=@n13.Out.L
    Medium.L=@n14.Out.L
    Close.L=@n17.Out.L
    Close.R=@n17.Out.R
    Suppressor.Close.L=@n19.Close.L
    Suppressor.Close.R=@n19.Close.R
    Suppressor.Medium=@n19.Medium.L
    ProfileFilter.Gain=VariableAccessor.Float(Variable=VariableMutator.Float(Variable=@n20.Variable, Value=@n1.RandomGainA).Variable).Value
    ProfileFilter.Bandwidth=VariableAccessor.Float(Variable=VariableMutator.Float(Variable=@n21.Variable, Value=@n1.RandomBandwidthA).Variable).Value
    ProfileFilter.CutoffFrequency=VariableAccessor.Float(Variable=VariableMutator.Float(Variable=@n22.Variable, Value=@n1.RandomFrequencyA).Variable).Value
    OnGunshot=VariableAccessor.Trigger(Variable=@n18.Variable).Value
   )
   Multiply_0 = UE.Multiply:Audio by Float(AdditionalOperands=^Mixing.WeaponOuputVolume, PrimaryOperand=@n23.Out.L)
   Multiply_1 = UE.Multiply:Audio by Float(AdditionalOperands=^Mixing.WeaponOuputVolume, PrimaryOperand=@n23.Out.R)
   n24 = VariableAccessor.Float(Variable=InitVariable.Float().Variable)
   AudioMixerAudioMixerMono2_0 = "AudioMixer.Audio Mixer (Mono, 2)"(In 0=@Multiply_0.Out, In 1=@Multiply_1.Out, Gain 0=@n24.Value, Gain 1=@n24.Value)
   AudioBusWriter1_0 = "UE.Audio Bus Writer (1)"(In 0=@AudioMixerAudioMixerMono2_0.Out) {also {defaults {Audio Bus=/Game/Sounds/Modulation/EnvelopeFollowers/FortWeapon/AB_WeaponFire}}}
   InitVariable.Float()
   InitVariable.Trigger()
   TriggerAccumulatorTriggerAccumulate4_0 = "TriggerAccumulator.Trigger Accumulate (4)"(In 2=@n13.OnFinished, In 1=@n14.OnFinished, In 0=@n17.OnFinished, In 3=@n19.OnFinished) {also {defaults {Auto Reset="<MetaSoundBuilderResult.SUCCEEDED: 0>"}}}
   n25 = &MSP_TriggerResolve(A=@TriggerAccumulatorTriggerAccumulate4_0.Out, B=@n9.Value)
   TriggerDelay_0 = UE.Trigger Delay(Delay Time=^RetriggerDelayTime, In=@TriggerAccumulatorTriggerAccumulate4_0.Out)
   TriggerControl_0 = UE.Trigger Control(Open=@n25.A, Close=@n25.B, Trigger In=@TriggerDelay_0.Out)
   ValueValue_0 = Value.Value:Bool(Set=@TriggerControl_0.Trigger Out)
   VariableMutator.Float(Variable=InitVariable.Float().Variable, Value=@n1.RandomBandwidthB)
   VariableMutator.Float(Variable=InitVariable.Float().Variable, Value=@n1.RandomGainB)
   VariableMutator.Float(Variable=InitVariable.Float().Variable, Value=@n1.MinPitchProfile)
   VariableMutator.Float(Variable=InitVariable.Float().Variable, Value=@n1.MaxPitchProfile)
   VariableMutator.Float(Variable=InitVariable.Float().Variable, Value=@n1.RandomFrequencyB)
   VariableMutator.Int32(Variable=InitVariable.Int32().Variable, Value=@n1.SeedProfile)
   VariableMutator.Int32(Variable=InitVariable.Int32().Variable, Value=@MSP_Weapon_Fire_Control_3p_0.CurrentAmmo)
  ]
  links [
   {dst="@OnGunshotOut", src=MSP_Weapon_Fire_Control_3p_0.OutGunshot}
   {dst="@UE.OutputFormat.Stereo.Audio:0", src=Multiply_0.Out}
   {dst="@UE.OutputFormat.Stereo.Audio:1", src=Multiply_1.Out}
   {dst="@UE.Source.OneShot.OnFinished", src=TriggerControl_0.Trigger Out}
   {dst="@bOnAlmostFinished", src=ValueValue_0.Output Value}
  ]
 }
 Mobile {
  nodes [
   n1 = &MSP_Weapon_Fire_RandomSeed(
    SeedProfile=^Character.UniqueID
    Random.EQ.A.MaxFreq=^Random.EQ.A.MaxFreq
    Random.EQ.A.MaxGain=^Random.EQ.A.MaxGain
    Random.EQ.A.MinFreq=^Random.EQ.A.MinFreq
    Random.EQ.A.MinGain=^Random.EQ.A.MinGain
    Random.EQ.B.MaxFreq=^Random.EQ.B.MaxFreq
    Random.EQ.B.MaxGain=^Random.EQ.B.MaxGain
    Random.EQ.B.MinFreq=^Random.EQ.B.MinFreq
    Random.EQ.B.MinGain=^Random.EQ.B.MinGain
    Random.MaxPitch=^Random.MaxPitch
    Random.MaxVolume=^Random.MaxVolume
    Random.MinPitch=^Random.MinPitch
    Random.MinVolume=^Random.MinVolume
    InitRandom=^UE.Source.OnPlay
   )
   n2 = VariableAccessor.Float(Variable=VariableMutator.Float(Variable=InitVariable.Float().Variable, Value=@n1.UniquePitch).Variable)
   n3 = VariableAccessor.Float(Variable=@n2.Variable)
   n4 = VariableAccessor.Float(Variable=@n3.Variable)
   n5 = VariableAccessor.Float(Variable=VariableMutator.Float(Variable=InitVariable.Float().Variable, Value=@n1.MinVolumeProfile).Variable)
   n6 = VariableAccessor.Float(Variable=@n5.Variable)
   n7 = VariableAccessor.Float(Variable=VariableMutator.Float(Variable=InitVariable.Float().Variable, Value=@n1.MaxVolumeProfile).Variable)
   n8 = VariableAccessor.Float(Variable=@n7.Variable)
   MSP_Weapon_Fire_Control_3p_0 = &MSP_Weapon_Fire_Control_3p(
    FireBurst=^FireBurst
    FireSingleShot=^FireSingleShot
    RoundsPerBurst=^RoundsPerBurst
    StartFiring=^StartFiring
    StopBurst=^StopBurst
    StopFiring=^StopFiring
    Weapon.AmmoRemaining=^Weapon.AmmoRemaining
    Weapon.FireRate=^Weapon.FireRate
   )
   n9 = VariableAccessor.Trigger(Variable=VariableMutator.Trigger(Variable=InitVariable.Trigger().Variable, Value=@MSP_Weapon_Fire_Control_3p_0.OutGunshot).Variable)
   n10 = VariableAccessor.Trigger(Variable=@n9.Variable)
   n11 = VariableAccessor.Trigger(Variable=@n10.Variable)
   n12 = VariableAccessor.Trigger(Variable=@n11.Variable)
   n13 = &MSP_Weapon_Fire_3p_Mono(Sounds=^MainFire.SoundsDistant, Random.Pitch=@n4.Value, Random.Volume.Min=@n6.Value, Random.Volume.Max=@n8.Value, OnGunshot=@n12.Value)
   n14 = &MSP_Weapon_Fire_3p_Mono(Sounds=^MainFire.SoundsMedium, Random.Volume.Max=@n7.Value, Random.Pitch=@n3.Value, Random.Volume.Min=@n5.Value, OnGunshot=@n11.Value)
   n15 = VariableAccessor.Float(Variable=@n6.Variable)
   n16 = VariableAccessor.Float(Variable=@n8.Variable)
   n17 = &MSP_Weapon_Fire_3p_Stereo(
    Sounds=^MainFire.SoundsClose
    Random.Volume.Min=VariableAccessor.Float(Variable=@n15.Variable).Value
    Random.Volume.Max=VariableAccessor.Float(Variable=@n16.Variable).Value
    Random.Pitch=@n2.Value
    OnGunshot=@n10.Value
   )
   n18 = VariableAccessor.Trigger(Variable=@n12.Variable)
   n19 = &MSP_Weapon_Fire_Suppressor_3p(
    Sounds.Close=^Suppressor.Sounds.Close
    Sounds.Medium=^Suppressor.Sounds.Medium
    Volume.Close=^Suppressor.Volume.Close
    Volume.Medium=^Suppressor.Volume.Medium
    bIsSuppressorEnabled=^Weapon.IsSuppressorEnabled
    Random.Volume.Min=@n15.Value
    Random.Pitch=VariableAccessor.Float(Variable=@n4.Variable).Value
    Random.Volume.Max=@n16.Value
    OnGunshot=@n18.Value
   )
   n20 = InitVariable.Float()
   n21 = InitVariable.Float()
   n22 = InitVariable.Float()
   n23 = &MSP_Weapon_Fire_Mixing_3p(
    Character.Affiliation=^Character.Affiliation
    bEnableDistanceEQ=^DistanceEQ.bEnable
    FireSingleShot=^FireSingleShot
    DistanceCurves=^MainFire.DistanceCurves
    MaxFalloffDistance=^MainFire.FalloffDistance
    Volume.Close=^Mixing.CloseGain
    Volume.Distant=^Mixing.DistantGain
    Mixing.Friendly=^Mixing.Friendly.FriendlyFireWeaponVolume
    Volume.Medium=^Mixing.MediumGain
    Suppressor.Mixing.SuppressorVolume=^Mixing.Suppressor.SuppressorOutputVolume
    Suppressor.Mixing.BaseWeaponVolume=^Suppressor.BaseWeaponVolumeMultiplier
    Suppressor.EnvelopeDecay=^Suppressor.EnvelopeDecayTime
    Distance=^UE.Attenuation.Distance
    Spatialization.Elevation=^UE.Spatialization.Elevation
    Weapon.FireRate=^Weapon.FireRate
    Weapon.IsSuppressorEnabled=^Weapon.IsSuppressorEnabled
    Distant.L=@n13.Out.L
    Medium.L=@n14.Out.L
    Close.L=@n17.Out.L
    Close.R=@n17.Out.R
    Suppressor.Close.L=@n19.Close.L
    Suppressor.Close.R=@n19.Close.R
    Suppressor.Medium=@n19.Medium.L
    ProfileFilter.Gain=VariableAccessor.Float(Variable=VariableMutator.Float(Variable=@n20.Variable, Value=@n1.RandomGainA).Variable).Value
    ProfileFilter.Bandwidth=VariableAccessor.Float(Variable=VariableMutator.Float(Variable=@n21.Variable, Value=@n1.RandomBandwidthA).Variable).Value
    ProfileFilter.CutoffFrequency=VariableAccessor.Float(Variable=VariableMutator.Float(Variable=@n22.Variable, Value=@n1.RandomFrequencyA).Variable).Value
    OnGunshot=VariableAccessor.Trigger(Variable=@n18.Variable).Value
   )
   Multiply_0 = UE.Multiply:Audio by Float(AdditionalOperands=^Mixing.WeaponOuputVolume, PrimaryOperand=@n23.Out.L)
   Multiply_1 = UE.Multiply:Audio by Float(AdditionalOperands=^Mixing.WeaponOuputVolume, PrimaryOperand=@n23.Out.R)
   n24 = VariableAccessor.Float(Variable=InitVariable.Float().Variable)
   AudioMixerAudioMixerMono2_0 = "AudioMixer.Audio Mixer (Mono, 2)"(In 0=@Multiply_0.Out, In 1=@Multiply_1.Out, Gain 0=@n24.Value, Gain 1=@n24.Value)
   AudioBusWriter1_0 = "UE.Audio Bus Writer (1)"(In 0=@AudioMixerAudioMixerMono2_0.Out) {also {defaults {Audio Bus=&AB_WeaponFire}}}
   InitVariable.Float()
   InitVariable.Trigger()
   TriggerAccumulatorTriggerAccumulate4_0 = "TriggerAccumulator.Trigger Accumulate (4)"(In 2=@n13.OnFinished, In 1=@n14.OnFinished, In 0=@n17.OnFinished, In 3=@n19.OnFinished) {also {defaults {Auto Reset="<MetaSoundBuilderResult.SUCCEEDED: 0>"}}}
   n25 = &MSP_TriggerResolve(A=@TriggerAccumulatorTriggerAccumulate4_0.Out, B=@n9.Value)
   TriggerDelay_0 = UE.Trigger Delay(Delay Time=^RetriggerDelayTime, In=@TriggerAccumulatorTriggerAccumulate4_0.Out)
   TriggerControl_0 = UE.Trigger Control(Open=@n25.A, Close=@n25.B, Trigger In=@TriggerDelay_0.Out)
   ValueValue_0 = Value.Value:Bool(Set=@TriggerControl_0.Trigger Out)
   VariableMutator.Float(Variable=InitVariable.Float().Variable, Value=@n1.RandomBandwidthB)
   VariableMutator.Float(Variable=InitVariable.Float().Variable, Value=@n1.RandomGainB)
   VariableMutator.Float(Variable=InitVariable.Float().Variable, Value=@n1.MinPitchProfile)
   VariableMutator.Float(Variable=InitVariable.Float().Variable, Value=@n1.MaxPitchProfile)
   VariableMutator.Float(Variable=InitVariable.Float().Variable, Value=@n1.RandomFrequencyB)
   VariableMutator.Int32(Variable=InitVariable.Int32().Variable, Value=@n1.SeedProfile)
   VariableMutator.Int32(Variable=InitVariable.Int32().Variable, Value=@MSP_Weapon_Fire_Control_3p_0.CurrentAmmo)
  ]
  links [
   {dst="@OnGunshotOut", src=MSP_Weapon_Fire_Control_3p_0.OutGunshot}
   {dst="@UE.OutputFormat.Stereo.Audio:0", src=Multiply_0.Out}
   {dst="@UE.OutputFormat.Stereo.Audio:1", src=Multiply_1.Out}
   {dst="@UE.Source.OneShot.OnFinished", src=TriggerControl_0.Trigger Out}
   {dst="@bOnAlmostFinished", src=ValueValue_0.Output Value}
  ]
 }
]
properties {
 assetImportData=&self:AssetImportData
 attenuationSettings=/FortMetaSoundWeapon/Items/Shared/Settings/FPS/AS_Weapon_FP_Fire_3P
 maxDistance=26200
 numChannels=2
 outputFormat=Stereo
 qualitySetting=Fort_Weapons
 sampleRate=48000
 soundClassObject=&Fort_WeaponRanged_Fire_3P_PSM
 concurrencyOverrides {bLimitToOwner=true, maxCount=2, resolutionRule=StopOldest}
 concurrencySet [
  /FortMetaSoundWeapon/Items/Shared/Settings/FPS/CS_Weapon_FP_Gunfire_Local
  /FortMetaSoundWeapon/Items/Shared/Settings/FPS/CS_Weapon_FP_Gunfire_Global
 ]
 modulationSettings {
  volumeRouting=Union
  volumeModulationDestination {bEnableModulation=true, modulators [/FortMetaSoundWeapon/Items/Modulation/PP_Weapon_Fire_3p_Vol]}
 }
}
